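# Control Rig for the F_MED_BlowWire female body cosmetic that corrects skirt, belt, pouch, knife, collar and glute bones as the character moves.
# It runs a shared DeformRig function, then derives movement and stance weights from input variables, pelvis height and calf distance.
# Weighted parent constraints set base placement.
# Spherical pose readers and bone deltas then weight additive transform offsets, some Kalman-smoothed, with extra branches for crouch, tactical sprint and emotes.
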ControlRigBlueprint /BRCosmetics/Characters/Player/Female/Medium/Bodies/F_MED_BlowWire/Meshes/F_MED_BlowWire_CtrlRig
graphs [
 Multiply_1_2_1_ContainedGraph {nodes [
   n1 = RigVMFunctionEntryNode()
   RigVMFunctionReturnNode(Result=&RigVMFunction_MathDoubleMul(B=@n1.D, A=&RigVMFunction_MathDoubleMul(A=@n1.A, B=@n1.C).Result).Result)
  ]}
 Multiply_3_1_2_3_ContainedGraph {nodes [
   n1 = RigVMFunctionEntryNode()
   RigVMFunctionReturnNode(Result=&RigVMFunction_MathDoubleMul(B=@n1.C, A=&RigVMFunction_MathDoubleMul(A=@n1.A, B=@n1.B).Result).Result)
  ]}
 Multiply_3_1_2_3_ContainedGraph {nodes [
   n1 = RigVMFunctionEntryNode()
   RigVMFunctionReturnNode(Result=&RigVMFunction_MathDoubleMul(B=@n1.C, A=&RigVMFunction_MathDoubleMul(A=@n1.A, B=@n1.B).Result).Result)
  ]}
 RigVMFunctionLibrary {links [], nodes []}
 RigVMFunction_Sequence_1_ContainedGraph {nodes [
   n1 = RigVMFunctionEntryNode()
   RigVMFunction_Sequence = &RigVMFunction_Sequence(ExecuteContext=@n1.ExecuteContext)
   n2 = &RigVMFunction_Sequence(ExecuteContext=@RigVMFunction_Sequence.B)
   n3 = &RigVMFunction_Sequence(ExecuteContext=@n2.B)
   n4 = &RigVMFunction_Sequence(ExecuteContext=@n3.B)
   n5 = &RigVMFunction_Sequence(ExecuteContext=@n4.B)
   n6 = &RigVMFunction_Sequence(ExecuteContext=@n5.B)
   n7 = &RigVMFunction_Sequence(ExecuteContext=@n6.B)
   n8 = &RigVMFunction_Sequence(ExecuteContext=@n7.B)
   RigVMFunctionReturnNode(A=@RigVMFunction_Sequence.A, B=@n2.A, C=@n3.A, D=@n4.A, E=@n5.A, F=@n6.A, G=@n7.A, H=@n8.A, I=@n8.B)
  ]}
 RigVMModel {nodes [
   RigUnit_BeginExecution = &RigUnit_BeginExecution()
   FN_DeformRig_FNC = RigVMFunctionReferenceNode(
    DeformRig_ Curve="(deformRig_l_upperarm_up_alpha_72_#1=False,deformRig_r_upperarm_up_alpha_73_#2=False,deformRig_l_upperarm_fr_alpha_74_#3=False,deformRig_r_upperarm_fr_alpha_75_CB51B8 ...(truncated)"
    DeformRig_Bool="(upperarm_1_#4=True,elbow_3_#5=True,wrist_7_#6=True,groin_8_#7=True,glute_17_#8=True,knee_9_40C408324CA261 ...(truncated)"
    DeformRig_Debug=(upperarm_debug_18_#4=False,wrist_debug_20_#6=False,lat_debug_26_#9=False,groinGlute_debug_27_#7=False)
    DeformRig_Value="(upperarm_up_18_#10=(Rotation=(X=0,Y=0,Z=0,W=1),Translation=(X=0,Y=0,Z=8),Scale3D=(X=0.7,Y=1,Z=1)),upperarm_fwd_fr_19_#11=(Rotation=(X=0,Y=0,Z=0,W=1),Translation=(X=0,Y=10,Z=0),Scale3D=(X=0.5,Y=1,Z= ...(truncated)"
    ExecuteContext=@RigUnit_BeginExecution.ExecutePin
   )
   > RigVMFunction_Sequence = RigVMAggregateNode()
   n1 = RigVMVariableNode() {variable=FB_Movement}
   RigVMVariableNode(
    Value=&RigVMFunction_MathDoubleRemap(SourceMaximum=1, SourceMinimum=0, TargetMaximum=1, TargetMinimum=0, Value=@n1.Value).Result
    ExecuteContext=@RigVMFunction_Sequence.B
   ) {variable=Forward_Movement}
   > RigVMVariableNode(Value=&RigVMFunction_MathDoubleRemap(SourceMaximum=0, SourceMinimum=-1, TargetMaximum=0, TargetMinimum=1, Value=@n1.Value).Result) {variable=Backward_Movement}
   n2 = RigVMVariableNode() {variable=LR_Movement}
   > RigVMVariableNode(Value=&RigVMFunction_MathDoubleRemap(SourceMaximum=0, SourceMinimum=-1, TargetMaximum=0, TargetMinimum=1, Value=@n2.Value).Result) {variable=Left_Movement}
   > RigVMVariableNode(Value=&RigVMFunction_MathDoubleRemap(SourceMaximum=1, SourceMinimum=0, TargetMaximum=1, TargetMinimum=0, Value=@n2.Value).Result) {variable=Right_Movement}
   RigUnit_GetTransform = &RigUnit_GetTransform(Item=(Type=Bone,Name="calf_l"), bInitial=False)
   RigVMFunction_MathVectorDistance = &RigVMFunction_MathVectorDistance(
    A=@RigUnit_GetTransform.Transform.Translation
    B=&RigUnit_GetTransform(Item=(Type=Bone,Name="calf_r"), bInitial=False).Transform.Translation
   )
   > RigVMVariableNode(Value=@RigVMFunction_MathVectorDistance.Result) {variable=Distance calf}
   Remap = &RigVMFunction_MathDoubleRemap(
    SourceMaximum=45
    SourceMinimum=55
    TargetMaximum=1
    TargetMinimum=0
    Value=&RigUnit_GetRelativeTransformForItem(Child=(Type=Bone,Name="pelvis"), Parent=(Type=Bone,Name="root"), bChildInitial=False, bParentInitial=False).RelativeTransform.Translation.Z
   )
   > RigVMVariableNode(Value=@Remap.Result) {variable=sit_lower}
   Multiply = &RigVMFunction_MathDoubleMul(
    B=&RigVMFunction_MathDoubleRemap(
     SourceMaximum=60
     SourceMinimum=50
     TargetMaximum=1
     TargetMinimum=0
     Value=&RigUnit_GetRelativeTransformForItem(Child=(Type=Bone,Name="pelvis"), Parent=(Type=Bone,Name="root"), bChildInitial=False, bParentInitial=False).RelativeTransform.Translation.Z
    ).Result
    A=&RigVMFunction_MathDoubleRemap(
     SourceMaximum=77
     SourceMinimum=85
     TargetMaximum=1
     TargetMinimum=0
     Value=&RigUnit_GetRelativeTransformForItem(Child=(Type=Bone,Name="pelvis"), Parent=(Type=Bone,Name="root"), bChildInitial=False, bParentInitial=False).RelativeTransform.Translation.Z
    ).Result
   )
   > RigVMVariableNode(Value=@Multiply.Result) {variable=crouch}
   > RigVMVariableNode(Value=&RigVMFunction_MathDoubleRemap(
     SourceMaximum=85
     SourceMinimum=80
     TargetMaximum=1
     TargetMinimum=0
     Value=&RigUnit_GetRelativeTransformForItem(Child=(Type=Bone,Name="pelvis"), Parent=(Type=Bone,Name="root"), bChildInitial=False, bParentInitial=False).RelativeTransform.Translation.Z
    ).Result) {variable=Stand}
   n3 = &RigUnit_GetRelativeTransformForItem(Child=(Type=Bone,Name="pelvis"), Parent=(Type=Bone,Name="root"), bChildInitial=False, bParentInitial=False)
   > RigVMVariableNode(Value=&RigVMFunction_KalmanFloat(
     BufferSize=16
     Value=&RigVMFunction_MathDoubleRemap(SourceMaximum=80, SourceMinimum=85, TargetMaximum=1, TargetMinimum=0, Value=@n3.RelativeTransform.Translation.Z).Result
    ) {template="KalmanFilter::Execute(in Value,in BufferSize,out Result)"}.Result) {variable=NoStand}
   &RigUnit_ParentConstraint(
    AdvancedSettings=(InterpolationType=Average,RotationOrderForFilter=XZY)
    Child=(Type=Bone,Name="Base_skirt_bk_01_l")
    Filter=(TranslationFilter=(bX=true,bY=true,bZ=true),RotationFilter=(bX=true,bY=true,bZ=true),ScaleFilter=(bX=false,bY=false,bZ=false))
    Parents=((Item=(Type=Bone,Name="pelvis"),Weight=0.7),(Item=(Type=Bone,Name="thigh_l"),Weight=0.3))
    Weight=1
    bMaintainOffset=true
    ExecutePin=@RigVMFunction_Sequence.A
   )
   > &RigUnit_ParentConstraint(
    AdvancedSettings=(InterpolationType=Average,RotationOrderForFilter=XZY)
    Child=(Type=Bone,Name="Base_skirt_bk_01_r")
    Filter=(TranslationFilter=(bX=true,bY=true,bZ=true),RotationFilter=(bX=True,bY=True,bZ=True),ScaleFilter=(bX=false,bY=false,bZ=false))
    Parents=((Item=(Type=Bone,Name="pelvis"),Weight=0.7),(Item=(Type=Bone,Name="thigh_r"),Weight=0.3))
    Weight=1
    bMaintainOffset=true
   )
   > &RigUnit_ParentConstraint(
    AdvancedSettings=(InterpolationType=Average,RotationOrderForFilter=XZY)
    Child=(Type=Bone,Name="Base_skirt_fr_01_l")
    Filter=(TranslationFilter=(bX=true,bY=true,bZ=true),RotationFilter=(bX=true,bY=true,bZ=true),ScaleFilter=(bX=false,bY=false,bZ=false))
    Parents=((Item=(Type=Bone,Name="pelvis"),Weight=0.3),(Item=(Type=Bone,Name="thigh_l"),Weight=0.7))
    Weight=1
    bMaintainOffset=true
   )
   > &RigUnit_ParentConstraint(
    AdvancedSettings=(InterpolationType=Average,RotationOrderForFilter=XZY)
    Child=(Type=Bone,Name="Base_skirt_fr_01_r")
    Filter=(TranslationFilter=(bX=true,bY=true,bZ=true),RotationFilter=(bX=True,bY=True,bZ=True),ScaleFilter=(bX=false,bY=false,bZ=false))
    Parents=((Item=(Type=Bone,Name="pelvis"),Weight=0.3),(Item=(Type=Bone,Name="thigh_r"),Weight=0.7))
    Weight=1
    bMaintainOffset=true
   )
   > &RigUnit_ParentConstraint(
    AdvancedSettings=(InterpolationType=Average,RotationOrderForFilter=XZY)
    Child=(Type=Bone,Name="dyn_belt_01_r_MSTR")
    Filter=(TranslationFilter=(bX=false,bY=false,bZ=false),RotationFilter=(bX=True,bY=True,bZ=True),ScaleFilter=(bX=false,bY=false,bZ=false))
    Parents=((Item=(Type=Bone,Name="pelvis"),Weight=0.3),(Item=(Type=Bone,Name="thigh_r"),Weight=0.7))
    Weight=1
    bMaintainOffset=true
   )
   > &RigUnit_ParentConstraint(
    AdvancedSettings=(InterpolationType=Average,RotationOrderForFilter=XZY)
    Child=(Type=Bone,Name="Base_skirt_side_bk_01_l")
    Filter=(TranslationFilter=(bX=true,bY=true,bZ=true),RotationFilter=(bX=True,bY=True,bZ=True),ScaleFilter=(bX=True,bY=True,bZ=True))
    Parents=((Item=(Type=Bone,Name="thigh_l"),Weight=0.6),(Item=(Type=Bone,Name="pelvis"),Weight=0.4))
    Weight=1
   )
   > &RigUnit_ParentConstraint(
    AdvancedSettings=(InterpolationType=Average,RotationOrderForFilter=XZY)
    Child=(Type=Bone,Name="Base_skirt_side_bk_01_r")
    Filter=(TranslationFilter=(bX=true,bY=true,bZ=true),RotationFilter=(bX=True,bY=True,bZ=True),ScaleFilter=(bX=True,bY=True,bZ=True))
    Parents=((Item=(Type=Bone,Name="thigh_r"),Weight=0.6),(Item=(Type=Bone,Name="pelvis"),Weight=0.4))
    Weight=1
   )
   > &RigUnit_ParentConstraint(
    AdvancedSettings=(InterpolationType=Average,RotationOrderForFilter=XZY)
    Child=(Type=Bone,Name="Base_skirt_side_fr_01_r")
    Filter=(TranslationFilter=(bX=true,bY=true,bZ=true),RotationFilter=(bX=True,bY=True,bZ=True),ScaleFilter=(bX=True,bY=True,bZ=True))
    Parents=((Item=(Type=Bone,Name="thigh_r"),Weight=0.9),(Item=(Type=Bone,Name="pelvis"),Weight=0.1))
    Weight=1
   )
   > &RigUnit_ParentConstraint(
    AdvancedSettings=(InterpolationType=Average,RotationOrderForFilter=XZY)
    Child=(Type=Bone,Name="Base_skirt_side_fr_01_l")
    Filter=(TranslationFilter=(bX=true,bY=true,bZ=true),RotationFilter=(bX=True,bY=True,bZ=True),ScaleFilter=(bX=True,bY=True,bZ=True))
    Parents=((Item=(Type=Bone,Name="thigh_l"),Weight=0.9),(Item=(Type=Bone,Name="pelvis"),Weight=0.1))
    Weight=1
   )
   > &RigUnit_ParentConstraint(
    AdvancedSettings=(InterpolationType=Average,RotationOrderForFilter=XZY)
    Child=(Type=Bone,Name="dyn_pouch_l_A_MSTR")
    Filter=(TranslationFilter=(bX=true,bY=true,bZ=true),RotationFilter=(bX=True,bY=True,bZ=True),ScaleFilter=(bX=True,bY=True,bZ=True))
    Parents=((Item=(Type=Bone,Name="spine_04"),Weight=0.0403),(Item=(Type=Bone,Name="spine_03"),Weight=0.0195),(Item=(Type=Bone,Name="spine_02"),Weight=0.3063),(Item=(Type=Bone,Name="spine_01"),Weight=0.0162),(Item=(Type=Bone,Name="pelvis"),Weight=0.6169))
    Weight=1
   )
   > &RigUnit_ParentConstraint(
    AdvancedSettings=(InterpolationType=Average,RotationOrderForFilter=XZY)
    Child=(Type=Bone,Name="dyn_pouch_l_B_MSTR")
    Filter=(TranslationFilter=(bX=true,bY=true,bZ=true),RotationFilter=(bX=True,bY=True,bZ=True),ScaleFilter=(bX=True,bY=True,bZ=True))
    Parents=((Item=(Type=Bone,Name="spine_04"),Weight=0.0438),(Item=(Type=Bone,Name="spine_03"),Weight=0.0238),(Item=(Type=Bone,Name="spine_02"),Weight=0.317),(Item=(Type=Bone,Name="spine_01"),Weight=0.0311),(Item=(Type=Bone,Name="pelvis"),Weight=0.5841))
    Weight=1
   )
   > &RigUnit_ParentConstraint(
    AdvancedSettings=(InterpolationType=Average,RotationOrderForFilter=XZY)
    Child=(Type=Bone,Name="dyn_knife_01_l_MSTR")
    Filter=(TranslationFilter=(bX=true,bY=true,bZ=true),RotationFilter=(bX=True,bY=True,bZ=True),ScaleFilter=(bX=True,bY=True,bZ=True))
    Parents=((Item=(Type=Bone,Name="spine_04"),Weight=0.0149),(Item=(Type=Bone,Name="spine_03"),Weight=0.017),(Item=(Type=Bone,Name="spine_02"),Weight=0.174),(Item=(Type=Bone,Name="spine_01"),Weight=0.0703),(Item=(Type=Bone,Name="pelvis"),Weight=0.7164))
    Weight=1
   )
   > &RigUnit_ParentConstraint(
    AdvancedSettings=(InterpolationType=Average,RotationOrderForFilter=XZY)
    Child=(Type=Bone,Name="foot_l_reparent")
    Filter=(TranslationFilter=(bX=true,bY=true,bZ=true),RotationFilter=(bX=True,bY=True,bZ=True),ScaleFilter=(bX=True,bY=True,bZ=True))
    Parents=((Item=(Type=Bone,Name="foot_l"),Weight=1))
    Weight=1
    bMaintainOffset=true
   )
   > &RigUnit_ParentConstraint(
    AdvancedSettings=(InterpolationType=Average,RotationOrderForFilter=XZY)
    Child=(Type=Bone,Name="ball_l_reparent")
    Filter=(TranslationFilter=(bX=true,bY=true,bZ=true),RotationFilter=(bX=True,bY=True,bZ=True),ScaleFilter=(bX=True,bY=True,bZ=True))
    Parents=((Item=(Type=Bone,Name="ball_l"),Weight=1))
    Weight=1
    bMaintainOffset=true
   )
   > &RigUnit_ParentConstraint(
    AdvancedSettings=(InterpolationType=Average,RotationOrderForFilter=XZY)
    Child=(Type=Bone,Name="foot_r_reparent")
    Filter=(TranslationFilter=(bX=true,bY=true,bZ=true),RotationFilter=(bX=True,bY=True,bZ=True),ScaleFilter=(bX=True,bY=True,bZ=True))
    Parents=((Item=(Type=Bone,Name="foot_r"),Weight=1))
    Weight=1
   )
   > &RigUnit_ParentConstraint(
    AdvancedSettings=(InterpolationType=Average,RotationOrderForFilter=XZY)
    Child=(Type=Bone,Name="ball_r_reparent")
    Filter=(TranslationFilter=(bX=true,bY=true,bZ=true),RotationFilter=(bX=True,bY=True,bZ=True),ScaleFilter=(bX=True,bY=True,bZ=True))
    Parents=((Item=(Type=Bone,Name="ball_r"),Weight=1))
    Weight=1
   )
   > &RigUnit_ModifyTransforms(
    ItemToModify=((Item=(Type=Bone,Name="Base_skirt_fr_01_r"),Transform=(Rotation=(X=0,Y=-0.0436,Z=0,W=0.999),Translation=(X=0,Y=0,Z=1),Scale3D=(X=1,Y=1,Z=1))))
    Weight=1
    WeightMinimum=0
   )
   > &RigUnit_ModifyTransforms(
    ItemToModify=((Item=(Type=Bone,Name="Base_skirt_fr_01_l"),Transform=(Rotation=(X=0,Y=-0.0436,Z=0,W=0.999),Translation=(X=0,Y=0,Z=1),Scale3D=(X=1,Y=1,Z=1))))
    Weight=1
    WeightMinimum=0
   )
   > &RigUnit_ModifyTransforms(
    ItemToModify=((Item=(Type=Bone,Name="Base_skirt_bk_01_l"),Transform=(Rotation=(X=0,Y=0,Z=0,W=1),Translation=(X=0,Y=0,Z=3),Scale3D=(X=1,Y=1,Z=1))))
    Weight=1
    WeightMinimum=0
   )
   > &RigUnit_ModifyTransforms(
    ItemToModify=((Item=(Type=Bone,Name="Base_skirt_bk_01_r"),Transform=(Rotation=(X=0,Y=0,Z=0,W=1),Translation=(X=0,Y=0,Z=3),Scale3D=(X=1,Y=1,Z=1))))
    Weight=1
    WeightMinimum=0
   )
   > &RigUnit_ModifyTransforms(
    ItemToModify=((Item=(Type=Bone,Name="Base_skirt_side_fr_01_l"),Transform=(Rotation=(X=0,Y=0,Z=0,W=1),Translation=(X=0,Y=1,Z=1),Scale3D=(X=1,Y=1,Z=1))))
    Weight=1
    WeightMinimum=0
   )
   > &RigUnit_ModifyTransforms(
    ItemToModify=((Item=(Type=Bone,Name="Base_skirt_side_fr_01_r"),Transform=(Rotation=(X=0,Y=0,Z=0,W=1),Translation=(X=0,Y=-1,Z=1),Scale3D=(X=1,Y=1,Z=1))))
    Weight=1
    WeightMinimum=0
   )
   > &RigUnit_ModifyTransforms(
    ItemToModify=((Item=(Type=Bone,Name="Base_skirt_side_bk_01_r"),Transform=(Rotation=(X=0,Y=0,Z=0,W=1),Translation=(X=0,Y=0,Z=1),Scale3D=(X=1,Y=1,Z=1))))
    Weight=1
    WeightMinimum=0
   )
   > &RigUnit_ModifyTransforms(
    ItemToModify=((Item=(Type=Bone,Name="Base_skirt_side_bk_01_l"),Transform=(Rotation=(X=0,Y=0,Z=0,W=1),Translation=(X=0,Y=0,Z=1),Scale3D=(X=1,Y=1,Z=1))))
    Weight=1
    WeightMinimum=0
   )
   > &RigUnit_ModifyTransforms(
    ItemToModify=((Item=(Type=Bone,Name="Dyn_collar_R"),Transform=(Rotation=(X=0,Y=0,Z=0,W=1),Translation=(X=0,Y=0,Z=1),Scale3D=(X=1,Y=1,Z=1))))
    Weight=1
    WeightMinimum=0
   )
   > &RigUnit_ModifyTransforms(
    ItemToModify=((Item=(Type=Bone,Name="Dyn_collar_L"),Transform=(Rotation=(X=0,Y=0,Z=0,W=1),Translation=(X=0,Y=0,Z=2),Scale3D=(X=1,Y=1,Z=1))))
    Weight=1
    WeightMinimum=0
   )
   Get Delta Ref_3 = RigVMFunctionReferenceNode(Item=(Type=Bone,Name="pelvis"), ExecuteContext=@RigVMFunction_Sequence.C)
   n4 = &RigVMFunction_MathDoubleMul(
    A=&RigVMFunction_MathDoubleRemap(SourceMaximum=80, SourceMinimum=90, TargetMaximum=1, TargetMinimum=0, Value=@"Get Delta Ref_3".Result.Rotation.Yaw).Result
    B=RigVMVariableNode() {variable=NoStand}.Value
   )
   > &RigUnit_ModifyTransforms(
    ItemToModify=((Item=(Type=Bone,Name="dyn_pouch_l_B"),Transform=(Rotation=(X=0.13,Y=-0.0114,Z=-0.0864,W=0.9877),Translation=(X=-0.6,Y=-4,Z=-1),Scale3D=(X=0.9,Y=1,Z=1))))
    WeightMinimum=0
    Weight=@n4.Result
   )
   > &RigUnit_ModifyTransforms(
    ItemToModify=((Item=(Type=Bone,Name="dyn_pouch_l_A"),Transform=(Rotation=(X=0.2549,Y=-0.0449,Z=-0.1677,W=0.9513),Translation=(X=2,Y=-4.5,Z=-1),Scale3D=(X=0.9,Y=1,Z=1))))
    WeightMinimum=0
    Weight=@n4.Result
   )
   > n5 = &RigUnit_SphericalPoseReader(
    ActiveRegionScaleFactors=(PositiveWidth=1,NegativeWidth=1,PositiveHeight=1,NegativeHeight=1)
    ActiveRegionSize=0.12
    Debug=(bDrawDebug=false,bDraw2D=False,bDrawLocalAxes=False,DebugScale=25,DebugSegments=20,DebugThickness=0.25)
    DriverAxis=(X=-1,Y=0,Z=0)
    DriverItem=(Type=Bone,Name="thigh_l")
    FalloffRegionScaleFactors=(PositiveWidth=1,NegativeWidth=1,PositiveHeight=1,NegativeHeight=1)
    FalloffSize=0.24
    FlipHeightScaling=False
    FlipWidthScaling=False
    OptionalParentItem=(Type=Bone,Name="None")
    RotationOffset=(X=115,Y=-30,Z=0)
   )
   n6 = &RigVMFunction_MathDoubleMul(A=@n5.OutputParam, B=RigVMVariableNode() {variable=Stand}.Value)
   > &RigUnit_ModifyTransforms(
    ItemToModify=((Item=(Type=Bone,Name="dyn_pouch_l_B_MSTR"),Transform=(Rotation=(X=0,Y=0,Z=0,W=1),Translation=(X=-1,Y=0,Z=0),Scale3D=(X=1,Y=1,Z=1))))
    WeightMinimum=0
    Weight=@n6.Result
   )
   > &RigUnit_ModifyTransforms(
    ItemToModify=((Item=(Type=Bone,Name="dyn_pouch_l_A_MSTR"),Transform=(Rotation=(X=0,Y=0,Z=0,W=1),Translation=(X=-2,Y=0,Z=0),Scale3D=(X=1,Y=1,Z=1))))
    WeightMinimum=0
    Weight=@n6.Result
   )
   > Get Delta Ref_2_1_1 = RigVMFunctionReferenceNode(Item=(Type=Bone,Name="pelvis"))
   n7 = &RigVMFunction_MathQuaternionToEuler(
    RotationOrder=XYZ
    Value=&RigUnit_GetRelativeTransformForItem(Child=(Type=Bone,Name="spine_04"), Parent=(Type=Bone,Name="spine_03"), bChildInitial=False, bParentInitial=False).RelativeTransform.Rotation
   )
   > &RigUnit_ModifyTransforms(
    ItemToModify=((Item=(Type=Bone,Name="dyn_knife_01_l_MSTR"),Transform=(Rotation=(X=0,Y=0,Z=-0.0436,W=0.999),Translation=(X=0,Y=-3,Z=0),Scale3D=(X=1,Y=1,Z=1))))
    WeightMinimum=0
    Weight=&RigVMFunction_MathDoubleMul(
     B=&RigVMFunction_MathDoubleRemap(SourceMaximum=-10, SourceMinimum=-5, TargetMaximum=1, TargetMinimum=0, Value=@n7.Result.Z).Result
     A=&RigVMFunction_MathDoubleRemap(SourceMaximum=70, SourceMinimum=60, TargetMaximum=1, TargetMinimum=0, Value=@"Get Delta Ref_2_1_1".Result.Rotation.Yaw).Result
    ).Result
   )
   > Get Delta Ref_2_1_1_1 = RigVMFunctionReferenceNode(Item=(Type=Bone,Name="pelvis"))
   > &RigUnit_ModifyTransforms(
    ItemToModify=((Item=(Type=Bone,Name="dyn_knife_01_l_MSTR"),Transform=(Rotation=(X=0,Y=0,Z=-0.0436,W=0.999),Translation=(X=0,Y=-1,Z=0.3),Scale3D=(X=1,Y=1,Z=1))))
    WeightMinimum=0
    Weight=&RigVMFunction_MathDoubleRemap(SourceMaximum=100, SourceMinimum=90, TargetMaximum=1, TargetMinimum=0, Value=@"Get Delta Ref_2_1_1_1".Result.Rotation.Yaw).Result
   )
   n8 = &RigUnit_GetRelativeTransformForItem(Child=(Type=Bone,Name="pelvis"), Parent=(Type=Bone,Name="root"), bChildInitial=False, bParentInitial=False)
   > &RigUnit_ModifyTransforms(
    ItemToModify="((Item=(Type=Bone,Name=\"Base_skirt_fr_01_l\"),Transform=(Rotation=(X=0,Y=0.1305,Z=0,W=0.9914),Translation=(X=0,Y=0,Z=0),Scale3D=(X=1,Y=1,Z=1))),(Item=(Type=Bone,Name=\"Base_skirt_side_bk_01_r\"),Transform=(Rotation=(X=0,Y=-0.0872,Z=0,W=0.9962),Translation=(X= ...(truncated)"
    WeightMinimum=0
    Weight=&RigVMFunction_KalmanFloat(
     BufferSize=16
     Value=&RigVMFunction_MathDoubleRemap(SourceMaximum=25, SourceMinimum=30, TargetMaximum=1, TargetMinimum=0, Value=@n8.RelativeTransform.Translation.Z).Result
    ) {template="KalmanFilter::Execute(in Value,in BufferSize,out Result)"}.Result
   )
   n9 = &RigVMFunction_ControlFlowBranch(Condition=RigVMVariableNode() {variable=TacticalSprinting}.Value, ExecuteContext=@RigVMFunction_Sequence.F)
   &RigUnit_ParentConstraint(
    AdvancedSettings=(InterpolationType=Average,RotationOrderForFilter=XZY)
    Child=(Type=Bone,Name="Base_skirt_fr_01_l")
    Filter=(TranslationFilter=(bX=true,bY=true,bZ=true),RotationFilter=(bX=true,bY=true,bZ=true),ScaleFilter=(bX=false,bY=false,bZ=false))
    Parents=((Item=(Type=Bone,Name="pelvis"),Weight=0),(Item=(Type=Bone,Name="thigh_l"),Weight=1))
    Weight=1
    bMaintainOffset=true
    ExecutePin=@n9.True
   )
   > &RigUnit_ParentConstraint(
    AdvancedSettings=(InterpolationType=Average,RotationOrderForFilter=XZY)
    Child=(Type=Bone,Name="Base_skirt_fr_01_r")
    Filter=(TranslationFilter=(bX=true,bY=true,bZ=true),RotationFilter=(bX=True,bY=True,bZ=True),ScaleFilter=(bX=false,bY=false,bZ=false))
    Parents=((Item=(Type=Bone,Name="pelvis"),Weight=0),(Item=(Type=Bone,Name="thigh_r"),Weight=1))
    Weight=1
    bMaintainOffset=true
   )
   > n10 = &RigUnit_SphericalPoseReader(
    ActiveRegionScaleFactors=(PositiveWidth=1,NegativeWidth=1,PositiveHeight=1,NegativeHeight=1)
    ActiveRegionSize=0.13
    Debug=(bDrawDebug=true,bDraw2D=False,bDrawLocalAxes=False,DebugScale=25,DebugSegments=20,DebugThickness=0.25)
    DriverAxis=(X=-1,Y=0,Z=0)
    DriverItem=(Type=Bone,Name="thigh_l")
    FalloffRegionScaleFactors=(PositiveWidth=1,NegativeWidth=1,PositiveHeight=1,NegativeHeight=1)
    FalloffSize=0.26
    FlipHeightScaling=False
    FlipWidthScaling=False
    OptionalParentItem=(Type=Bone,Name="None")
    RotationOffset=(X=125,Y=0,Z=35)
   )
   n11 = &RigVMFunction_MathDoubleMul(A=@n10.OutputParam, B=RigVMVariableNode() {variable=Stand}.Value)
   > &RigUnit_ModifyTransforms(
    ItemToModify=((Item=(Type=Bone,Name="Base_skirt_side_fr_01_l"),Transform=(Rotation=(X=0,Y=-0.0872,Z=0,W=0.9962),Translation=(X=-1,Y=0,Z=1),Scale3D=(X=1,Y=1,Z=1))))
    WeightMinimum=0
    Weight=@n11.Result
   )
   > &RigUnit_ModifyTransforms(
    ItemToModify=((Item=(Type=Bone,Name="Base_skirt_side_bk_01_l"),Transform=(Rotation=(X=0,Y=-0.1305,Z=0,W=0.9914),Translation=(X=-1,Y=0,Z=1),Scale3D=(X=1,Y=1,Z=1))))
    WeightMinimum=0
    Weight=@n11.Result
   )
   n12 = &RigUnit_SphericalPoseReader(
    ActiveRegionScaleFactors=(PositiveWidth=1,NegativeWidth=1,PositiveHeight=1,NegativeHeight=1)
    ActiveRegionSize=0.15
    Debug=(bDrawDebug=false,bDraw2D=False,bDrawLocalAxes=False,DebugScale=25,DebugSegments=20,DebugThickness=0.25)
    DriverAxis=(X=1,Y=0,Z=0)
    DriverItem=(Type=Bone,Name="thigh_r")
    FalloffRegionScaleFactors=(PositiveWidth=1,NegativeWidth=1,PositiveHeight=1,NegativeHeight=1)
    FalloffSize=0.3
    FlipHeightScaling=False
    FlipWidthScaling=False
    OptionalParentItem=(Type=Bone,Name="None")
    RotationOffset=(X=-85,Y=0,Z=30)
    ExecutePin=@RigVMFunction_Sequence.D
   )
   n13 = RigVMVariableNode() {variable=Stand}
   n14 = RigVMVariableNode() {variable=Forward_Movement}
   Multiply_3_1_2 = RigVMAggregateNode(A=@n12.OutputParam, B=@n13.Value, C=@n14.Value)
   > &RigUnit_ModifyTransforms(
    ItemToModify="((Item=(Type=Bone,Name=\"Base_skirt_fr_01_r\"),Transform=(Rotation=(X=0,Y=-0.1305,Z=0,W=0.9914),Translation=(X=0,Y=0,Z=0),Scale3D=(X=1,Y=1,Z=1))),(Item=(Type=Bone,Name=\"Base_skirt_side_fr_01_r\"),Transform=(Rotation=(X=0,Y=-0.0872,Z=0,W=0.9962),Translation=(X ...(truncated)"
    WeightMinimum=0
    Weight=@Multiply_3_1_2.Result
   )
   > n15 = &RigUnit_SphericalPoseReader(
    ActiveRegionScaleFactors=(PositiveWidth=1,NegativeWidth=1,PositiveHeight=1,NegativeHeight=1)
    ActiveRegionSize=0.15
    Debug=(bDrawDebug=false,bDraw2D=False,bDrawLocalAxes=False,DebugScale=25,DebugSegments=20,DebugThickness=0.25)
    DriverAxis=(X=-1,Y=0,Z=0)
    DriverItem=(Type=Bone,Name="thigh_l")
    FalloffRegionScaleFactors=(PositiveWidth=1,NegativeWidth=1,PositiveHeight=1,NegativeHeight=1)
    FalloffSize=0.3
    FlipHeightScaling=False
    FlipWidthScaling=False
    OptionalParentItem=(Type=Bone,Name="None")
    RotationOffset=(X=95,Y=0,Z=30)
   )
   Multiply_3_1_2_3 = RigVMAggregateNode(A=@n15.OutputParam, B=@n13.Value, C=@n14.Value)
   > &RigUnit_ModifyTransforms(
    ItemToModify="((Item=(Type=Bone,Name=\"Base_skirt_fr_01_l\"),Transform=(Rotation=(X=-0.0076,Y=-0.1735,Z=-0.043,W=0.9839),Translation=(X=0,Y=0,Z=0),Scale3D=(X=1,Y=1,Z=1))),(Item=(Type=Bone,Name=\"dyn_skirt_side_fr_01_l\"),Transform=(Rotation=(X=0.0038,Y=-0.0871,Z=0.0435,W=0. ...(truncated)"
    WeightMinimum=0
    Weight=@Multiply_3_1_2_3.Result
   )
   > n16 = &RigUnit_ModifyTransforms(
    ItemToModify="((Item=(Type=Bone,Name=\"Base_skirt_fr_01_l\"),Transform=(Rotation=(X=-0.1006,Y=-0.1648,Z=-0.1006,W=0.976),Translation=(X=2,Y=0,Z=2),Scale3D=(X=1,Y=1,Z=1))),(Item=(Type=Bone,Name=\"Base_skirt_side_fr_01_l\"),Transform=(Rotation=(X=0,Y=-0.0436,Z=0,W=0.999),Tran ...(truncated)"
    WeightMinimum=0
    Weight=RigVMVariableNode() {variable=sit_lower}.Value
   )
   Greater = /Script/RigVM.RigVMFunction_MathDoubleGreater(B=0, A=RigVMVariableNode() {variable=crouch}.Value) {is=struct RigVMUnitNode, template="Greater::Execute(in A,in B,out Result)"}
   n17 = &RigVMFunction_ControlFlowBranch(Condition=@Greater.Result, ExecuteContext=@n16.ExecutePin)
   n18 = &RigUnit_SphericalPoseReader(
    ActiveRegionScaleFactors=(PositiveWidth=1,NegativeWidth=1,PositiveHeight=1,NegativeHeight=1)
    ActiveRegionSize=0.11
    Debug=(bDrawDebug=false,bDraw2D=False,bDrawLocalAxes=False,DebugScale=25,DebugSegments=20,DebugThickness=0.25)
    DriverAxis=(X=1,Y=0,Z=0)
    DriverItem=(Type=Bone,Name="thigh_r")
    FalloffRegionScaleFactors=(PositiveWidth=1,NegativeWidth=1,PositiveHeight=1,NegativeHeight=1)
    FalloffSize=0.22
    FlipHeightScaling=False
    FlipWidthScaling=False
    OptionalParentItem=(Type=Bone,Name="None")
    RotationOffset=(X=-75,Y=0,Z=10)
    ExecutePin=@n17.True
   )
   > &RigUnit_ModifyTransforms(
    ItemToModify=((Item=(Type=Bone,Name="dyn_belt_01_r_MSTR"),Transform=(Rotation=(X=0,Y=-0.1305,Z=0,W=0.9914),Translation=(X=0,Y=0,Z=1),Scale3D=(X=1,Y=1,Z=1))))
    WeightMinimum=0
    Weight=@n18.OutputParam
   )
   > n19 = &RigUnit_SphericalPoseReader(
    ActiveRegionScaleFactors=(PositiveWidth=1,NegativeWidth=1,PositiveHeight=1,NegativeHeight=1)
    ActiveRegionSize=0.13
    Debug=(bDrawDebug=true,bDraw2D=False,bDrawLocalAxes=False,DebugScale=25,DebugSegments=20,DebugThickness=0.25)
    DriverAxis=(X=1,Y=0,Z=0)
    DriverItem=(Type=Bone,Name="thigh_r")
    FalloffRegionScaleFactors=(PositiveWidth=1,NegativeWidth=1,PositiveHeight=1,NegativeHeight=1)
    FalloffSize=0.26
    FlipHeightScaling=False
    FlipWidthScaling=False
    OptionalParentItem=(Type=Bone,Name="None")
    RotationOffset=(X=-75,Y=0,Z=35)
   )
   KalmanFilter = &RigVMFunction_KalmanFloat(
    BufferSize=8
    Value=&RigVMFunction_MathDoubleMul(A=@n19.OutputParam, B=RigVMVariableNode() {variable=Forward_Movement}.Value).Result
   ) {template="KalmanFilter::Execute(in Value,in BufferSize,out Result)"}
   > &RigUnit_ModifyTransforms(
    ItemToModify="((Item=(Type=Bone,Name=\"Base_skirt_side_bk_01_r\"),Transform=(Rotation=(X=0.0076,Y=-0.0868,Z=0.0868,W=0.9924),Translation=(X=0,Y=1,Z=0),Scale3D=(X=1,Y=1,Z=1))),(Item=(Type=Bone,Name=\"Base_skirt_fr_01_r\"),Transform=(Rotation=(X=0,Y=-0.0436,Z=0,W=0.999),Trans ...(truncated)"
    WeightMinimum=0
    Weight=@KalmanFilter.Result
   )
   > n20 = &RigUnit_SphericalPoseReader(
    ActiveRegionScaleFactors=(PositiveWidth=1,NegativeWidth=1,PositiveHeight=1,NegativeHeight=1)
    ActiveRegionSize=0.1
    Debug=(bDrawDebug=false,bDraw2D=False,bDrawLocalAxes=False,DebugScale=25,DebugSegments=20,DebugThickness=0.25)
    DriverAxis=(X=-1,Y=0,Z=0)
    DriverItem=(Type=Bone,Name="thigh_l")
    FalloffRegionScaleFactors=(PositiveWidth=1,NegativeWidth=1,PositiveHeight=1,NegativeHeight=1)
    FalloffSize=0.22
    FlipHeightScaling=False
    FlipWidthScaling=False
    OptionalParentItem=(Type=Bone,Name="None")
    RotationOffset=(X=95,Y=0,Z=0)
   )
   > &RigUnit_ModifyTransforms(
    ItemToModify="((Item=(Type=Bone,Name=\"Base_skirt_fr_01_l\"),Transform=(Rotation=(X=0,Y=-0.1305,Z=0,W=0.9914),Translation=(X=0,Y=0,Z=2),Scale3D=(X=1,Y=1,Z=1))),(Item=(Type=Bone,Name=\"Base_skirt_side_fr_01_l\"),Transform=(Rotation=(X=0,Y=-0.0436,Z=0,W=0.999),Translation=(X= ...(truncated)"
    WeightMinimum=0
    Weight=&RigVMFunction_KalmanFloat(BufferSize=8, Value=@n20.OutputParam) {template="KalmanFilter::Execute(in Value,in BufferSize,out Result)"}.Result
   )
   Add = &RigVMFunction_MathDoubleAdd(A=RigVMVariableNode() {variable=Left_Movement}.Value, B=RigVMVariableNode() {variable=Right_Movement}.Value)
   n21 = &RigVMFunction_MathDoubleRemap(SourceMaximum=55, SourceMinimum=45, TargetMaximum=1, TargetMinimum=0, Value=RigVMVariableNode() {variable=Distance calf}.Value)
   > &RigUnit_ModifyTransforms(
    ItemToModify="((Item=(Type=Bone,Name=\"Base_skirt_side_fr_01_l\"),Transform=(Rotation=(X=0,Y=-0.0872,Z=0,W=0.9962),Translation=(X=0,Y=0,Z=1),Scale3D=(X=1,Y=1,Z=1))),(Item=(Type=Bone,Name=\"Base_skirt_side_fr_01_r\"),Transform=(Rotation=(X=0,Y=-0.1305,Z=0,W=0.9914),Translati ...(truncated)"
    WeightMinimum=0
    Weight=&RigVMFunction_KalmanFloat(BufferSize=8, Value=&RigVMFunction_MathDoubleMul(B=@Add.Result, A=@n21.Result).Result) {template="KalmanFilter::Execute(in Value,in BufferSize,out Result)"}.Result
   )
   > &RigUnit_SphericalPoseReader(
    ActiveRegionScaleFactors=(PositiveWidth=1,NegativeWidth=1,PositiveHeight=1,NegativeHeight=1)
    ActiveRegionSize=0.1
    Debug=(bDrawDebug=true,bDraw2D=False,bDrawLocalAxes=False,DebugScale=25,DebugSegments=20,DebugThickness=0.25)
    DriverAxis=(X=1,Y=0,Z=0)
    DriverItem=(Type=Bone,Name="thigh_r")
    FalloffRegionScaleFactors=(PositiveWidth=1,NegativeWidth=1,PositiveHeight=1,NegativeHeight=1)
    FalloffSize=0.2
    FlipHeightScaling=False
    FlipWidthScaling=False
    OptionalParentItem=(Type=Bone,Name="None")
    RotationOffset=(X=-65,Y=0,Z=20)
   )
   > Get Delta Ref_1 = RigVMFunctionReferenceNode(Item=(Type=Bone,Name="thigh_r"))
   Multiply_1_2 = RigVMAggregateNode(
    A=&RigVMFunction_MathDoubleRemap(SourceMaximum=-10, SourceMinimum=-20, TargetMaximum=1, TargetMinimum=0, Value=@"Get Delta Ref_1".Result.Rotation.Roll).Result
    C=&RigVMFunction_MathDoubleRemap(SourceMaximum=10, SourceMinimum=20, TargetMaximum=1, TargetMinimum=0, Value=@"Get Delta Ref_1".Result.Rotation.Roll).Result
    D=RigVMVariableNode() {variable=Right_Movement}.Value
   )
   > &RigUnit_ModifyTransforms(
    ItemToModify=((Item=(Type=Bone,Name="Base_skirt_fr_01_l"),Transform=(Rotation=(X=0,Y=-0.1305,Z=0,W=0.9914),Translation=(X=0,Y=0,Z=1),Scale3D=(X=1,Y=1,Z=1))))
    WeightMinimum=0
    Weight=&RigVMFunction_KalmanFloat(BufferSize=8, Value=@Multiply_1_2.Result) {template="KalmanFilter::Execute(in Value,in BufferSize,out Result)"}.Result
   )
   RigVMFunction_ControlFlowBranch = &RigVMFunction_ControlFlowBranch(Condition=RigVMVariableNode() {variable=Emotes}.Value, ExecuteContext=@RigVMFunction_Sequence.E)
   n22 = &RigUnit_SphericalPoseReader(
    ActiveRegionScaleFactors=(PositiveWidth=1,NegativeWidth=1,PositiveHeight=1,NegativeHeight=1)
    ActiveRegionSize=0.13
    Debug=(bDrawDebug=false,bDraw2D=False,bDrawLocalAxes=False,DebugScale=25,DebugSegments=20,DebugThickness=0.25)
    DriverAxis=(X=1,Y=0,Z=0)
    DriverItem=(Type=Bone,Name="thigh_r")
    FalloffRegionScaleFactors=(PositiveWidth=1,NegativeWidth=1,PositiveHeight=1,NegativeHeight=1)
    FalloffSize=0.26
    FlipHeightScaling=False
    FlipWidthScaling=False
    OptionalParentItem=(Type=Bone,Name="None")
    RotationOffset=(X=-30,Y=10,Z=0)
    ExecutePin=@RigVMFunction_ControlFlowBranch.True
   )
   > n23 = &RigUnit_SphericalPoseReader(
    ActiveRegionScaleFactors=(PositiveWidth=1,NegativeWidth=1,PositiveHeight=1,NegativeHeight=1)
    ActiveRegionSize=0.13
    Debug=(bDrawDebug=false,bDraw2D=False,bDrawLocalAxes=False,DebugScale=25,DebugSegments=20,DebugThickness=0.25)
    DriverAxis=(X=-1,Y=0,Z=0)
    DriverItem=(Type=Bone,Name="thigh_l")
    FalloffRegionScaleFactors=(PositiveWidth=1,NegativeWidth=1,PositiveHeight=1,NegativeHeight=1)
    FalloffSize=0.26
    FlipHeightScaling=False
    FlipWidthScaling=False
    OptionalParentItem=(Type=Bone,Name="None")
    RotationOffset=(X=155,Y=10,Z=0)
   )
   n24 = &RigVMFunction_MathDoubleMul(B=@n22.OutputParam, A=@n23.OutputParam)
   > &RigUnit_ModifyTransforms(
    ItemToModify=((Item=(Type=Bone,Name="Base_skirt_fr_01_l"),Transform=(Rotation=(X=0,Y=-0.1736,Z=0,W=0.9848),Translation=(X=0,Y=0,Z=2),Scale3D=(X=1,Y=1,Z=1))))
    WeightMinimum=0
    Weight=@n24.Result
   )
   > &RigUnit_ModifyTransforms(
    ItemToModify=((Item=(Type=Bone,Name="Base_skirt_side_fr_01_l"),Transform=(Rotation=(X=0,Y=-0.0436,Z=0,W=0.999),Translation=(X=0,Y=1,Z=2),Scale3D=(X=1,Y=1,Z=1))))
    WeightMinimum=0
    Weight=@n24.Result
   )
   > &RigUnit_ModifyTransforms(
    ItemToModify=((Item=(Type=Bone,Name="Base_skirt_side_bk_01_l"),Transform=(Rotation=(X=0.1261,Y=-0.2566,Z=0.0338,W=0.9577),Translation=(X=0,Y=2,Z=2),Scale3D=(X=1,Y=1,Z=1))))
    WeightMinimum=0
    Weight=@n24.Result
   )
   > &RigUnit_ModifyTransforms(
    ItemToModify=((Item=(Type=Bone,Name="Base_skirt_side_fr_01_r"),Transform=(Rotation=(X=0,Y=-0.0436,Z=0,W=0.999),Translation=(X=0,Y=1,Z=2),Scale3D=(X=1,Y=1,Z=1))))
    WeightMinimum=0
    Weight=@n24.Result
   )
   > &RigUnit_ModifyTransforms(
    ItemToModify=((Item=(Type=Bone,Name="Base_skirt_fr_01_r"),Transform=(Rotation=(X=0,Y=0,Z=0.0872,W=0.9962),Translation=(X=0,Y=0,Z=1),Scale3D=(X=1,Y=1,Z=1))))
    WeightMinimum=0
    Weight=@n24.Result
   )
   > &RigUnit_ModifyTransforms(
    ItemToModify=((Item=(Type=Bone,Name="Base_skirt_side_bk_01_r"),Transform=(Rotation=(X=-0.0762,Y=-0.1701,Z=-0.055,W=0.9809),Translation=(X=0,Y=-1,Z=2),Scale3D=(X=1,Y=1,Z=1))))
    WeightMinimum=0
    Weight=@n24.Result
   )
   > &RigUnit_ModifyTransforms(
    ItemToModify=((Item=(Type=Bone,Name="dyn_pouch_l_B_MSTR"),Transform=(Rotation=(X=0,Y=0,Z=-0.1736,W=0.9848),Translation=(X=0,Y=-2,Z=0),Scale3D=(X=1,Y=1,Z=1))))
    WeightMinimum=0
    Weight=@n24.Result
   )
   > &RigUnit_ModifyTransforms(
    ItemToModify=((Item=(Type=Bone,Name="dyn_pouch_l_A_MSTR"),Transform=(Rotation=(X=0,Y=0,Z=-0.1736,W=0.9848),Translation=(X=0,Y=-2,Z=0),Scale3D=(X=1,Y=1,Z=1))))
    WeightMinimum=0
    Weight=@n24.Result
   )
   > Get Delta Ref_2_3 = RigVMFunctionReferenceNode(Item=(Type=Bone,Name="thigh_l"))
   n25 = &RigVMFunction_MathDoubleRemap(SourceMaximum=35, SourceMinimum=25, TargetMaximum=1, TargetMinimum=0, Value=@"Get Delta Ref_2_3".Result.Rotation.Roll)
   > &RigUnit_ModifyTransforms(
    ItemToModify=((Item=(Type=Bone,Name="deform_glute_l"),Transform=(Rotation=(X=0,Y=0,Z=0,W=1),Translation=(X=-1.5,Y=-1.5,Z=0),Scale3D=(X=1,Y=1,Z=1))))
    WeightMinimum=0
    Weight=@n25.Result
   )
   > &RigUnit_ModifyTransforms(
    ItemToModify=((Item=(Type=Bone,Name="Base_skirt_bk_01_l"),Transform=(Rotation=(X=0,Y=0,Z=0,W=1),Translation=(X=0,Y=0,Z=2),Scale3D=(X=1,Y=1,Z=1))))
    WeightMinimum=0
    Weight=@n25.Result
   )
   > &RigUnit_ModifyTransforms(
    ItemToModify="((Item=(Type=Bone,Name=\"Base_skirt_side_bk_01_l\"),Transform=(Rotation=(X=0,Y=-0.0872,Z=0,W=0.9962),Translation=(X=0,Y=0.5,Z=1.5),Scale3D=(X=1,Y=1,Z=1))),(Item=(Type=Bone,Name=\"Base_skirt_side_fr_01_l\"),Transform=(Rotation=(X=0,Y=0,Z=0,W=1),Translation=(X=0 ...(truncated)"
    WeightMinimum=0
    Weight=@n25.Result
   )
   > Get Delta Ref_2_1_3 = RigVMFunctionReferenceNode(Item=(Type=Bone,Name="thigh_r"))
   > &RigUnit_ModifyTransforms(
    ItemToModify=((Item=(Type=Bone,Name="Base_skirt_side_bk_01_r"),Transform=(Rotation=(X=0,Y=0,Z=0,W=1),Translation=(X=0,Y=0,Z=2),Scale3D=(X=1,Y=1,Z=1))))
    WeightMinimum=0
    Weight=&RigVMFunction_MathDoubleRemap(SourceMaximum=40, SourceMinimum=30, TargetMaximum=1, TargetMinimum=0, Value=@"Get Delta Ref_2_1_3".Result.Rotation.Roll).Result
   )
   > &RigUnit_ModifyTransforms(
    ItemToModify=((Item=(Type=Bone,Name="deform_glute_r"),Transform=(Rotation=(X=0,Y=0,Z=0,W=1),Translation=(X=1,Y=1,Z=0),Scale3D=(X=1,Y=1,Z=1))))
    Weight=1
    WeightMinimum=0
   )
   n26 = &RigVMFunction_ControlFlowBranch(Condition=RigVMVariableNode() {variable=Emotes}.Value, ExecuteContext=@RigVMFunction_Sequence.G)
   n27 = &RigUnit_SphericalPoseReader(
    ActiveRegionScaleFactors=(PositiveWidth=1,NegativeWidth=1,PositiveHeight=1,NegativeHeight=1)
    ActiveRegionSize=0.13
    Debug=(bDrawDebug=false,bDraw2D=False,bDrawLocalAxes=False,DebugScale=25,DebugSegments=20,DebugThickness=0.25)
    DriverAxis=(X=1,Y=0,Z=0)
    DriverItem=(Type=Bone,Name="thigh_r")
    FalloffRegionScaleFactors=(PositiveWidth=1,NegativeWidth=1,PositiveHeight=1,NegativeHeight=1)
    FalloffSize=0.26
    FlipHeightScaling=False
    FlipWidthScaling=False
    OptionalParentItem=(Type=Bone,Name="None")
    RotationOffset=(X=-90,Y=0,Z=-25)
    ExecutePin=@n26.True
   )
   > &RigUnit_ModifyTransforms(
    ItemToModify=((Item=(Type=Bone,Name="Base_skirt_fr_01_r"),Transform=(Rotation=(X=0.0151,Y=-0.0858,Z=0.173,W=0.9811),Translation=(X=0,Y=0,Z=4),Scale3D=(X=1,Y=1,Z=1))))
    WeightMinimum=0
    Weight=@n27.OutputParam
   )
   > &RigUnit_ModifyTransforms(
    ItemToModify=((Item=(Type=Bone,Name="Base_skirt_side_fr_01_r"),Transform=(Rotation=(X=0,Y=0,Z=-0.0872,W=0.9962),Translation=(X=0,Y=-1,Z=0),Scale3D=(X=1,Y=1,Z=1))))
    WeightMinimum=0
    Weight=@n27.OutputParam
   )
   > &RigUnit_ModifyTransforms(
    ItemToModify=((Item=(Type=Bone,Name="dyn_belt_01_r_MSTR"),Transform=(Rotation=(X=0,Y=-0.2588,Z=0,W=0.9659),Translation=(X=0,Y=0,Z=2),Scale3D=(X=1,Y=1,Z=1))))
    WeightMinimum=0
    Weight=@n27.OutputParam
   )
   > n28 = &RigUnit_SphericalPoseReader(
    ActiveRegionScaleFactors=(PositiveWidth=1,NegativeWidth=1,PositiveHeight=1,NegativeHeight=1)
    ActiveRegionSize=0.1
    Debug=(bDrawDebug=false,bDraw2D=False,bDrawLocalAxes=False,DebugScale=25,DebugSegments=20,DebugThickness=0.25)
    DriverAxis=(X=-1,Y=0,Z=0)
    DriverItem=(Type=Bone,Name="thigh_l")
    FalloffRegionScaleFactors=(PositiveWidth=1,NegativeWidth=1,PositiveHeight=1,NegativeHeight=1)
    FalloffSize=0.2
    FlipHeightScaling=False
    FlipWidthScaling=False
    OptionalParentItem=(Type=Bone,Name="None")
    RotationOffset=(X=110,Y=0,Z=0)
   )
   > &RigUnit_ModifyTransforms(
    ItemToModify=((Item=(Type=Bone,Name="deform_glute_r"),Transform=(Rotation=(X=0,Y=0,Z=0,W=1),Translation=(X=1,Y=0,Z=0),Scale3D=(X=1,Y=1,Z=1))))
    WeightMinimum=0
    Weight=@n28.OutputParam
   )
   > n29 = &RigUnit_SphericalPoseReader(
    ActiveRegionScaleFactors=(PositiveWidth=1,NegativeWidth=1,PositiveHeight=1,NegativeHeight=1)
    ActiveRegionSize=0.13
    Debug=(bDrawDebug=false,bDraw2D=False,bDrawLocalAxes=False,DebugScale=25,DebugSegments=20,DebugThickness=0.25)
    DriverAxis=(X=-1,Y=0,Z=0)
    DriverItem=(Type=Bone,Name="clavicle_r")
    FalloffRegionScaleFactors=(PositiveWidth=1,NegativeWidth=1,PositiveHeight=1,NegativeHeight=1)
    FalloffSize=0.26
    FlipHeightScaling=False
    FlipWidthScaling=False
    OptionalParentItem=(Type=Bone,Name="None")
    RotationOffset=(X=180,Y=-40,Z=0)
   )
   > n30 = &RigUnit_SphericalPoseReader(
    ActiveRegionScaleFactors=(PositiveWidth=1,NegativeWidth=1,PositiveHeight=1,NegativeHeight=1)
    ActiveRegionSize=0.13
    Debug=(bDrawDebug=false,bDraw2D=False,bDrawLocalAxes=False,DebugScale=25,DebugSegments=20,DebugThickness=0.25)
    DriverAxis=(X=1,Y=0,Z=0)
    DriverItem=(Type=Bone,Name="upperarm_l")
    FalloffRegionScaleFactors=(PositiveWidth=1,NegativeWidth=1,PositiveHeight=1,NegativeHeight=1)
    FalloffSize=0.26
    FlipHeightScaling=False
    FlipWidthScaling=False
    OptionalParentItem=(Type=Bone,Name="None")
    RotationOffset=(X=-10,Y=-30,Z=0)
   )
   n31 = &RigVMFunction_MathDoubleMul(B=@n29.OutputParam, A=@n30.OutputParam)
   > &RigUnit_ModifyTransforms(
    ItemToModify=((Item=(Type=Bone,Name="Dyn_collar_R"),Transform=(Rotation=(X=0,Y=-0.0872,Z=0,W=0.9962),Translation=(X=0,Y=0,Z=0),Scale3D=(X=1,Y=1,Z=1))))
    WeightMinimum=0
    Weight=@n31.Result
   )
   > &RigUnit_ModifyTransforms(
    ItemToModify=((Item=(Type=Bone,Name="Dyn_collar_L"),Transform=(Rotation=(X=0,Y=-0.1736,Z=0,W=0.9848),Translation=(X=0,Y=0,Z=0),Scale3D=(X=1,Y=1,Z=1))))
    WeightMinimum=0
    Weight=@n31.Result
   )
   > n32 = &RigUnit_SphericalPoseReader(
    ActiveRegionScaleFactors=(PositiveWidth=1,NegativeWidth=1,PositiveHeight=1,NegativeHeight=1)
    ActiveRegionSize=0.1
    Debug=(bDrawDebug=false,bDraw2D=False,bDrawLocalAxes=False,DebugScale=25,DebugSegments=20,DebugThickness=0.25)
    DriverAxis=(X=-1,Y=0,Z=0)
    DriverItem=(Type=Bone,Name="thigh_l")
    FalloffRegionScaleFactors=(PositiveWidth=1,NegativeWidth=1,PositiveHeight=1,NegativeHeight=1)
    FalloffSize=0.22
    FlipHeightScaling=False
    FlipWidthScaling=False
    OptionalParentItem=(Type=Bone,Name="None")
    RotationOffset=(X=110,Y=0,Z=55)
   )
   RigVMFunction_MathQuaternionToEuler = &RigVMFunction_MathQuaternionToEuler(
    RotationOrder=ZYX
    Value=&RigUnit_GetRelativeTransformForItem(Child=(Type=Bone,Name="spine_05"), Parent=(Type=Bone,Name="pelvis"), bChildInitial=False, bParentInitial=False).RelativeTransform.Rotation
   )
   n33 = &RigVMFunction_MathDoubleMul(
    B=&RigVMFunction_MathDoubleRemap(SourceMaximum=35, SourceMinimum=25, TargetMaximum=1, TargetMinimum=0, Value=@RigVMFunction_MathQuaternionToEuler.Result.Y).Result
    A=@n32.OutputParam
   )
   > &RigUnit_ModifyTransforms(
    ItemToModify=((Item=(Type=Bone,Name="dyn_pouch_l_A"),Transform=(Rotation=(X=0,Y=-0.1305,Z=0,W=0.9914),Translation=(X=-2.5,Y=0,Z=1),Scale3D=(X=1,Y=1,Z=1))))
    WeightMinimum=0
    Weight=@n33.Result
   )
   > &RigUnit_ModifyTransforms(
    ItemToModify=((Item=(Type=Bone,Name="dyn_pouch_l_B"),Transform=(Rotation=(X=0,Y=-0.0872,Z=0,W=0.9962),Translation=(X=-0.7,Y=0,Z=-1),Scale3D=(X=1,Y=1,Z=1))))
    WeightMinimum=0
    Weight=@n33.Result
   )
   > &RigUnit_ModifyTransforms(
    ItemToModify=((Item=(Type=Bone,Name="Base_skirt_fr_01_l"),Transform=(Rotation=(X=0,Y=-0.0872,Z=0,W=0.9962),Translation=(X=0,Y=0,Z=0),Scale3D=(X=1,Y=1,Z=1))))
    WeightMinimum=0
    Weight=@n33.Result
   )
   > Get Delta Ref_2 = RigVMFunctionReferenceNode(Item=(Type=Bone,Name="foot_l"))
   > Get Delta Ref_2_1 = RigVMFunctionReferenceNode(Item=(Type=Bone,Name="foot_r"))
   n34 = &RigVMFunction_MathDoubleRemap(SourceMaximum=50, SourceMinimum=40, TargetMaximum=1, TargetMinimum=0, Value=@"Get Delta Ref_2".Result.Rotation.Yaw)
   > &RigUnit_ModifyTransforms(
    ItemToModify=((Item=(Type=Bone,Name="foot_l_reparent"),Transform=(Rotation=(X=0,Y=0,Z=-0.0436,W=0.999),Translation=(X=-4,Y=1,Z=0),Scale3D=(X=1,Y=1,Z=1))))
    WeightMinimum=0
    Weight=@n34.Result
   )
   n35 = &RigVMFunction_MathDoubleRemap(SourceMaximum=40, SourceMinimum=25, TargetMaximum=1, TargetMinimum=0, Value=@"Get Delta Ref_2_1".Result.Rotation.Yaw)
   > &RigUnit_ModifyTransforms(
    ItemToModify=((Item=(Type=Bone,Name="foot_r_reparent"),Transform=(Rotation=(X=0,Y=0,Z=0,W=1),Translation=(X=2,Y=1,Z=0),Scale3D=(X=1,Y=1,Z=1))))
    WeightMinimum=0
    Weight=@n35.Result
   )
   > &RigUnit_ModifyTransforms(
    ItemToModify="((Item=(Type=Bone,Name=\"Base_skirt_side_bk_01_r\"),Transform=(Rotation=(X=0,Y=-0.1305,Z=0,W=0.9914),Translation=(X=0,Y=0,Z=0),Scale3D=(X=1,Y=1,Z=1))),(Item=(Type=Bone,Name=\"dyn_skirt_side_fr_01_r\"),Transform=(Rotation=(X=0,Y=-0.1305,Z=0,W=0.9914),Translatio ...(truncated)"
    WeightMinimum=0
    Weight=&RigVMFunction_MathDoubleMul(B=@n34.Result, A=@n35.Result).Result
   )
   n36 = &RigVMFunction_ControlFlowBranch(Condition=RigVMVariableNode() {variable=RdanOFF}.Value, ExecuteContext=@RigVMFunction_Sequence.H)
   /Script/ControlRig.RigUnit_AimItem(
    DebugSettings=(bEnabled=False,Scale=10,WorldOffset=(Rotation=(X=0,Y=0,Z=0,W=1),Translation=(X=0,Y=0,Z=0),Scale3D=(X=1,Y=1,Z=1)))
    Item=(Type=Bone,Name="dyn_belt_01_r")
    Primary=(Weight=1,Axis=(X=1,Y=0,Z=0),Target=(X=1,Y=0,Z=0),Kind=Direction,Space=(Type=Bone,Name="thigh_r"))
    Secondary=(Weight=1,Axis=(X=0,Y=0,Z=1),Target=(X=0,Y=0,Z=1),Kind=Location,Space=(Type=Bone,Name="dyn_belt_01_r_MSTR"))
    Weight=1
    ExecutePin=@n36.True
   ) {is=struct RigVMUnitNode}
   > &RigUnit_ModifyTransforms(
    ItemToModify=((Item=(Type=Bone,Name="dyn_belt_01_r"),Transform=(Rotation=(X=0.0298,Y=-0.0819,Z=0.3407,W=0.9361),Translation=(X=0,Y=0,Z=0),Scale3D=(X=1,Y=1,Z=1))))
    Weight=1
    WeightMinimum=0
   )
   > n37 = &RigUnit_SphericalPoseReader(
    ActiveRegionScaleFactors=(PositiveWidth=1,NegativeWidth=1,PositiveHeight=1,NegativeHeight=1)
    ActiveRegionSize=0.13
    Debug=(bDrawDebug=false,bDraw2D=False,bDrawLocalAxes=False,DebugScale=25,DebugSegments=20,DebugThickness=0.25)
    DriverAxis=(X=1,Y=0,Z=0)
    DriverItem=(Type=Bone,Name="thigh_r")
    FalloffRegionScaleFactors=(PositiveWidth=1,NegativeWidth=1,PositiveHeight=1,NegativeHeight=1)
    FalloffSize=0.26
    FlipHeightScaling=False
    FlipWidthScaling=False
    OptionalParentItem=(Type=Bone,Name="None")
    RotationOffset=(X=-75,Y=0,Z=35)
   )
   > &RigUnit_ModifyTransforms(
    ItemToModify=((Item=(Type=Bone,Name="dyn_belt_01_r"),Transform=(Rotation=(X=0,Y=0,Z=0,W=1),Translation=(X=0,Y=0,Z=2),Scale3D=(X=1,Y=1,Z=1))))
    WeightMinimum=0
    Weight=&RigVMFunction_MathDoubleMul(A=@n37.OutputParam, B=RigVMVariableNode() {variable=crouch}.Value).Result
   )
   > &RigUnit_ModifyTransforms(
    ItemToModify=((Item=(Type=Bone,Name="dyn_belt_01_r"),Transform=(Rotation=(X=0,Y=0.1736,Z=0,W=0.9848),Translation=(X=0,Y=0,Z=0),Scale3D=(X=1,Y=1,Z=1))))
    WeightMinimum=0
    Weight=RigVMVariableNode() {variable=sit_lower}.Value
   )
   > n38 = &RigUnit_SphericalPoseReader(
    ActiveRegionScaleFactors=(PositiveWidth=1,NegativeWidth=1,PositiveHeight=1,NegativeHeight=1)
    ActiveRegionSize=0.11
    Debug=(bDrawDebug=true,bDraw2D=False,bDrawLocalAxes=False,DebugScale=25,DebugSegments=20,DebugThickness=0.25)
    DriverAxis=(X=1,Y=0,Z=0)
    DriverItem=(Type=Bone,Name="thigh_r")
    FalloffRegionScaleFactors=(PositiveWidth=1,NegativeWidth=1,PositiveHeight=1,NegativeHeight=1)
    FalloffSize=0.24
    FlipHeightScaling=False
    FlipWidthScaling=False
    OptionalParentItem=(Type=Bone,Name="None")
    RotationOffset=(X=90,Y=0,Z=-70)
   )
   > n39 = &RigUnit_ModifyTransforms(
    ItemToModify=((Item=(Type=Bone,Name="dyn_belt_01_r_MSTR"),Transform=(Rotation=(X=0,Y=-0.1736,Z=0,W=0.9848),Translation=(X=0,Y=0,Z=0),Scale3D=(X=1,Y=1,Z=1))))
    WeightMinimum=0
    Weight=&RigVMFunction_MathDoubleMul(A=@n38.OutputParam, B=RigVMVariableNode() {variable=Stand}.Value).Result
   )
   n40 = &RigVMFunction_Sequence(ExecuteContext=@n39.ExecutePin)
   n41 = &RigVMFunction_ControlFlowBranch(ExecuteContext=@n40.A, Condition=RigVMVariableNode() {variable=Lod1}.Value)
   &RigUnit_ModifyTransforms(
    ItemToModify=((Item=(Type=Bone,Name="dyn_skirt_bk_01_r"),Transform=(Rotation=(X=0,Y=0,Z=0,W=1),Translation=(X=0,Y=0,Z=3),Scale3D=(X=1,Y=1,Z=1))))
    Weight=1
    WeightMinimum=0
    ExecutePin=@n41.True
   )
   > &RigUnit_ModifyTransforms(
    ItemToModify=((Item=(Type=Bone,Name="dyn_skirt_bk_01_l"),Transform=(Rotation=(X=0,Y=0,Z=0,W=1),Translation=(X=0,Y=0,Z=2),Scale3D=(X=1,Y=1,Z=1))))
    Weight=1
    WeightMinimum=0
   )
   > &RigUnit_ModifyTransforms(
    ItemToModify=((Item=(Type=Bone,Name="correct_ALT_groin_R"),Transform=(Rotation=(X=0,Y=0,Z=0,W=1),Translation=(X=0,Y=-4,Z=0),Scale3D=(X=1,Y=1,Z=1))))
    Weight=1
    WeightMinimum=0
   )
   > &RigUnit_ModifyTransforms(
    ItemToModify=((Item=(Type=Bone,Name="correct_ALT_groin_L"),Transform=(Rotation=(X=0,Y=0,Z=0,W=1),Translation=(X=0,Y=3,Z=0),Scale3D=(X=1,Y=1,Z=1))))
    Weight=1
    WeightMinimum=0
   )
   n42 = &RigVMFunction_ControlFlowBranch(ExecuteContext=@n40.B, Condition=RigVMVariableNode() {variable=Lod2}.Value)
   &RigUnit_ModifyTransforms(
    ItemToModify=((Item=(Type=Bone,Name="dyn_skirt_bk_01_r"),Transform=(Rotation=(X=0,Y=0,Z=0,W=1),Translation=(X=0,Y=-2,Z=3),Scale3D=(X=1,Y=1,Z=1))))
    Weight=1
    WeightMinimum=0
    ExecutePin=@n42.True
   )
   > &RigUnit_ModifyTransforms(
    ItemToModify=((Item=(Type=Bone,Name="dyn_skirt_bk_01_l"),Transform=(Rotation=(X=0,Y=0,Z=0,W=1),Translation=(X=0,Y=2,Z=3),Scale3D=(X=1,Y=1,Z=1))))
    Weight=1
    WeightMinimum=0
   )
   > &RigUnit_ModifyTransforms(
    ItemToModify=((Item=(Type=Bone,Name="correct_ALT_groin_R"),Transform=(Rotation=(X=0,Y=0,Z=0,W=1),Translation=(X=0,Y=-10,Z=5),Scale3D=(X=1,Y=1,Z=1))))
    Weight=1
    WeightMinimum=0
   )
   > &RigUnit_ModifyTransforms(
    ItemToModify=((Item=(Type=Bone,Name="correct_ALT_groin_L"),Transform=(Rotation=(X=0,Y=0,Z=0,W=1),Translation=(X=0,Y=10,Z=-8),Scale3D=(X=1,Y=1,Z=1))))
    Weight=1
    WeightMinimum=0
   )
   > &RigUnit_ModifyTransforms(
    ItemToModify=((Item=(Type=Bone,Name="Correct_ALT_belt_C"),Transform=(Rotation=(X=0,Y=0,Z=0,W=1),Translation=(X=0,Y=0,Z=0),Scale3D=(X=1,Y=1.2,Z=1.2))))
    Weight=1
    WeightMinimum=0
   )
   > &RigUnit_ModifyTransforms(
    ItemToModify=((Item=(Type=Bone,Name="thigh_Correct_ALT_R"),Transform=(Rotation=(X=0,Y=0,Z=0,W=1),Translation=(X=0,Y=0,Z=0),Scale3D=(X=1,Y=0.7,Z=0.7))))
    Weight=1
    WeightMinimum=0
   )
   > &RigUnit_ModifyTransforms(
    ItemToModify=((Item=(Type=Bone,Name="thigh_Correct_ALT_L"),Transform=(Rotation=(X=0,Y=0,Z=0,W=1),Translation=(X=0,Y=0,Z=0),Scale3D=(X=1,Y=0.7,Z=0.7))))
    Weight=1
    WeightMinimum=0
   )
   &RigUnit_ModifyTransforms(
    ExecutePin=()
    ItemToModify="((Item=(Type=Bone,Name=\"Base_skirt_fr_01_l\"),Transform=(Rotation=(X=-0.173,Y=0.0151,Z=-0.0858,W=0.9811),Translation=(X=2,Y=0,Z=2.5),Scale3D=(X=1,Y=1,Z=1))),(Item=(Type=Bone,Name=\"Base_skirt_side_fr_01_l\"),Transform=(Rotation=(X=0,Y=-0.0436,Z=0,W=0.999),Tra ...(truncated)"
    Weight=1
    WeightMinimum=0
   )
   &RigVMFunction_MathDoubleRemap(SourceMaximum=50, SourceMinimum=40, TargetMaximum=1, TargetMinimum=0, Value=0)
  ]}
]
hierarchy [
 root {loc=<0 0 0>, rot=<0 0 0 1>}
 pelvis {parent=root, loc=<0 -0.5301 98.0571>, rot=<0.0127 0.707 -0.0127 -0.707>}
 spine_01 {parent=pelvis, loc=<6.1208 -0.9328 0>, rot=<0 0 0.034 -0.9994>}
 spine_02 {parent=spine_01, loc=<6.5997 -0.2095 0>, rot=<0 0 0.0238 -0.9997>}
 spine_03 {parent=spine_02, loc=<6.4914 -0.1237 0>, rot=<0 0 0.0149 -0.9999>}
 spine_04 {parent=spine_03, loc=<6.9598 0.0794 0>, rot=<0 0 -0.0132 -0.9999>}
 spine_05 {parent=spine_04, loc=<6.9987 0 0>, rot=<0 0 -0.0015 -1>}
 attach_cape {parent=spine_05, loc=<9.3216 -8.4501 0>, rot=<0.0616 -0.7044 -0.0616 -0.7044>}
 neck_01 {parent=spine_05, loc=<11.3138 0.1529 0>, rot=<0 0 -0.2003 -0.9797>}
 neck_02 {parent=neck_01, loc=<4.5627 0 0>, rot=<0 0 0.0504 -0.9987>}
 head {parent=neck_02, loc=<5.3914 0 0>, rot=<0 0 0.115 -0.9934>}
 FACIAL_C_FacialRoot {parent=head, loc=<1.728 0.5445 0>, rot=<0.5021 -0.4979 -0.5021 -0.4979>}
 FACIAL_C_UnderChin {parent=FACIAL_C_FacialRoot, loc=<0 8.412 3.0294>, rot=<0.5 0 0 -0.866>}
 FACIAL_L_UnderChin {parent=FACIAL_C_FacialRoot, loc=<2.6853 7.0393 1.8393>, rot=<0.4025 -0.355 0.0823 -0.8398>}
 FACIAL_R_UnderChin {parent=FACIAL_C_FacialRoot, loc=<-2.6853 7.0393 1.8393>, rot=<0.4025 0.355 -0.0823 -0.8398>}
 FACIAL_C_Jaw {parent=FACIAL_C_FacialRoot, loc=<-0.0001 1.701 1.0343>, rot=<0.342 0 0 -0.9397>}
 FACIAL_C_Jawline {parent=FACIAL_C_Jaw, loc=<0.0001 2.3758 7.5718>, rot=<0.2588 0 0 -0.9659>}
 FACIAL_L_Jawline {parent=FACIAL_C_FacialRoot, loc=<4.1416 5.1709 3.393>, rot=<0.2092 0.1578 -0.0547 -0.9635>}
 FACIAL_R_Jawline {parent=FACIAL_C_FacialRoot, loc=<-4.1416 5.1709 3.393>, rot=<0.2092 -0.1578 0.0547 -0.9635>}
 neck_02_front {parent=neck_02, loc=<0 2 0>, rot=<0 0 0 1>}
 ring_01_MSTR {parent=neck_01, loc=<2.3715 5.0197 0>, rot=<-0.1608 -0.987 0 0>}
 ring_01 {parent=ring_01_MSTR, loc=<0 0 0>, rot=<0 0 0 1>}
 attach_backpack {parent=spine_05, loc=<0 -8.35 0>, rot=<-0.0093 -0.707 0.0093 -0.707>}
 clavicle_l {parent=spine_05, loc=<5.4831 -0.1053 -3.2946>, rot=<0.0465 -0.7289 -0.009 -0.6829>}
 upperarm_l {parent=clavicle_l, loc=<9.8658 -1.2136 -0.4105>, rot=<0.0164 -0.3245 -0.0269 -0.9454>}
 lowerarm_l {parent=upperarm_l, loc=<26.465 -0.8345 0.259>, rot=<0.0036 0.0161 -0.2298 -0.9731>}
 hand_l {parent=lowerarm_l, loc=<22.7355 -0.3203 -0.6105>, rot=<-0.6784 -0.0223 0.0735 -0.7307>}
 thumb_01_l {parent=hand_l, loc=<1.3118 -0.3191 -2.459>, rot=<0.6546 -0.4042 0.0664 -0.6354>}
 thumb_02_l {parent=thumb_01_l, loc=<5.9336 -0.0448 0.0401>, rot=<-0.0037 0.0213 0.298 -0.9543>}
 thumb_03_l {parent=thumb_02_l, loc=<2.2223 0.3682 -0.0907>, rot=<0.0118 -0.013 0.0066 -0.9998>}
 index_metacarpal_l {parent=hand_l, loc=<1.6535 0.0183 -2.96>, rot=<0.0158 -0.0124 -0.0173 -0.9996>}
 index_01_l {parent=index_metacarpal_l, loc=<8.3008 0.0357 -0.0524>, rot=<-0.0044 0.0485 0.0924 -0.9945>}
 index_02_l {parent=index_01_l, loc=<3.3587 -0.3004 0.0452>, rot=<-0.002 0.0027 0.3225 -0.9466>}
 index_03_l {parent=index_02_l, loc=<2.5578 -0.123 -0.0252>, rot=<-0.002 0.0067 0.2558 -0.9667>}
 middle_metacarpal_l {parent=hand_l, loc=<2.1856 -0.1725 -0.5984>, rot=<0.006 0.0168 -0.0329 -0.9993>}
 middle_01_l {parent=middle_metacarpal_l, loc=<7.3275 -0.2447 0.0129>, rot=<-0.0208 0.0042 0.0858 -0.9961>}
 middle_02_l {parent=middle_01_l, loc=<4.0123 0.1401 0.1175>, rot=<-0.0218 0.0258 0.3512 -0.9357>}
 middle_03_l {parent=middle_02_l, loc=<2.6077 0.0534 0.1373>, rot=<0.0082 0.0188 0.1708 -0.9851>}
 ring_metacarpal_l {parent=hand_l, loc=<2.2074 -0.3625 1.3638>, rot=<-0.0284 0.0722 -0.0458 -0.9959>}
 ring_01_l {parent=ring_metacarpal_l, loc=<7.0295 -0.6197 -0.1446>, rot=<-0.0489 -0.0252 0.1892 -0.9804>}
 ring_02_l {parent=ring_01_l, loc=<3.6174 0.4647 -0.1027>, rot=<0 0 0.2806 -0.9598>}
 ring_03_l {parent=ring_02_l, loc=<2.4152 0.114 0.0199>, rot=<0 0 0.1189 -0.9929>}
 pinky_metacarpal_l {parent=hand_l, loc=<1.7616 -0.9753 3.3814>, rot=<-0.0397 0.0814 -0.069 -0.9935>}
 pinky_01_l {parent=pinky_metacarpal_l, loc=<7.1709 -0.4831 -0.1094>, rot=<-0.0654 -0.0035 0.2242 -0.9723>}
 pinky_02_l {parent=pinky_01_l, loc=<2.8274 0.0893 -0.1043>, rot=<0.0037 0.0011 0.2854 -0.9584>}
 pinky_03_l {parent=pinky_02_l, loc=<2.2046 0.0307 -0.0078>, rot=<-0.0002 0.0098 0.1445 -0.9895>}
 deform_wrist_dn_base_l {parent=hand_l, loc=<0 -1.5326 0>, rot=<0 0 0 -1>}
 deform_wrist_dn_l {parent=deform_wrist_dn_base_l, loc=<0 0 0>, rot=<0 0 0 1>}
 deform_wrist_up_base_l {parent=hand_l, loc=<0 2.3009 0>, rot=<0 0 0 -1>}
 deform_wrist_up_l {parent=deform_wrist_up_base_l, loc=<0 0 0>, rot=<0 0 0 1>}
 hand_attach_l {parent=hand_l, loc=<8.5 -3 0>, rot=<0.7071 0.7071 0 0>}
 weapon_l {parent=hand_l, loc=<-0.1974 -3.4111 0.3811>, rot=<0 0 0 1>}
 lowerarm_twist_02_l {parent=lowerarm_l, loc=<10.6707 0 0>, rot=<0 0 0 1>}
 dyn_sleeve_dn_l {parent=lowerarm_twist_02_l, loc=<0.8917 -0.3006 -5.999>, rot=<0 -0.1518 0 -0.9884>}
 dyn_sleeve_bk_l {parent=lowerarm_twist_02_l, loc=<1.4963 -5.826 -0.0554>, rot=<-0.6535 0.0653 0.075 -0.7503>}
 dyn_sleeve_ft_l {parent=lowerarm_twist_02_l, loc=<1.2479 3.5221 0.2804>, rot=<0.7071 0 0 -0.7071>}
 dyn_sleeve_up_l {parent=lowerarm_twist_02_l, loc=<0.8188 -0.4137 3.6993>, rot=<0 0.0166 0 -0.9999>}
 lowerarm_twist_01_l {parent=lowerarm_l, loc=<18.316 0 0>, rot=<0 0 0 1>}
 deform_elbow_in_base_l {parent=lowerarm_l, loc=<0.1834 3.6062 -0.2572>, rot=<0 0 0 -1>}
 deform_elbow_in_l {parent=deform_elbow_in_base_l, loc=<0 0 0>, rot=<0 0 0 1>}
 deform_elbow_out_base_l {parent=lowerarm_l, loc=<0 -4 0>, rot=<0 0 0 1>}
 deform_elbow_out_l {parent=deform_elbow_out_base_l, loc=<0 0 0>, rot=<0 0 0 1>}
 upperarm_twist_01_l {parent=upperarm_l, loc=<7.2135 -0.1134 0>, rot=<0 0 0 1>}
 deform_upperarm_bk_base_l {parent=upperarm_twist_01_l, loc=<-6.4053 -4.3746 0.314>, rot=<0 0 0 -1>}
 deform_upperarm_bk_l {parent=deform_upperarm_bk_base_l, loc=<0 0 0>, rot=<0 0 0 1>}
 deform_upperarm_fr_base_l {parent=upperarm_twist_01_l, loc=<-5.6312 5.092 0.3354>, rot=<0 0 0 -1>}
 deform_upperarm_fr_l {parent=deform_upperarm_fr_base_l, loc=<0 0 0>, rot=<0 0 0 1>}
 deform_upperarm_up_base_l {parent=upperarm_twist_01_l, loc=<-8.891 0.4825 3.9517>, rot=<0 0 0 -1>}
 deform_upperarm_up_l {parent=deform_upperarm_up_base_l, loc=<0 0 0>, rot=<0 0 0 1>}
 upperarm_twist_02_l {parent=upperarm_l, loc=<19.7998 -0.1134 0>, rot=<0 0 0 1>}
 base_shoulders_l {parent=clavicle_l, loc=<12.8911 0.2319 6.5943>, rot=<0.6855 -0.2206 -0.215 -0.6597>}
 shoulders_l {parent=base_shoulders_l, loc=<0 0 0>, rot=<0 0 0 1>}
 clavicle_r {parent=spine_05, loc=<5.4831 -0.1053 3.2946>, rot=<0.7289 0.0465 -0.6829 0.009>}
 upperarm_r {parent=clavicle_r, loc=<-9.8658 1.2136 0.4105>, rot=<0.0164 -0.3245 -0.0269 -0.9454>}
 lowerarm_r {parent=upperarm_r, loc=<-26.465 0.8345 -0.259>, rot=<0.0036 0.0161 -0.2298 -0.9731>}
 hand_r {parent=lowerarm_r, loc=<-22.7355 0.3203 0.6106>, rot=<-0.6784 -0.0223 0.0735 -0.7307>}
 thumb_01_r {parent=hand_r, loc=<-1.3118 0.3191 2.459>, rot=<0.6546 -0.4042 0.0664 -0.6354>}
 thumb_02_r {parent=thumb_01_r, loc=<-5.9336 0.0448 -0.0401>, rot=<-0.0037 0.0213 0.298 -0.9543>}
 thumb_03_r {parent=thumb_02_r, loc=<-2.2223 -0.3682 0.0907>, rot=<0.0118 -0.013 0.0066 -0.9998>}
 index_metacarpal_r {parent=hand_r, loc=<-1.6535 -0.0183 2.96>, rot=<0.0158 -0.0124 -0.0173 -0.9996>}
 index_01_r {parent=index_metacarpal_r, loc=<-8.3008 -0.0357 0.0524>, rot=<-0.0044 0.0485 0.0924 -0.9945>}
 index_02_r {parent=index_01_r, loc=<-3.3587 0.3004 -0.0452>, rot=<-0.002 0.0027 0.3225 -0.9466>}
 index_03_r {parent=index_02_r, loc=<-2.5578 0.123 0.0252>, rot=<-0.002 0.0067 0.2558 -0.9667>}
 middle_metacarpal_r {parent=hand_r, loc=<-2.1856 0.1725 0.5984>, rot=<0.006 0.0168 -0.0329 -0.9993>}
 middle_01_r {parent=middle_metacarpal_r, loc=<-7.3275 0.2447 -0.0129>, rot=<-0.0208 0.0042 0.0858 -0.9961>}
 middle_02_r {parent=middle_01_r, loc=<-4.0123 -0.1401 -0.1175>, rot=<-0.0218 0.0258 0.3512 -0.9357>}
 middle_03_r {parent=middle_02_r, loc=<-2.6077 -0.0534 -0.1373>, rot=<0.0082 0.0188 0.1708 -0.9851>}
 ring_metacarpal_r {parent=hand_r, loc=<-2.2074 0.3625 -1.3638>, rot=<-0.0284 0.0722 -0.0458 -0.9959>}
 ring_01_r {parent=ring_metacarpal_r, loc=<-7.0295 0.6197 0.1446>, rot=<-0.0489 -0.0252 0.1892 -0.9804>}
 ring_02_r {parent=ring_01_r, loc=<-3.6174 -0.4647 0.1027>, rot=<0 0 0.2806 -0.9598>}
 ring_03_r {parent=ring_02_r, loc=<-2.4152 -0.114 -0.0199>, rot=<0 0 0.1189 -0.9929>}
 pinky_metacarpal_r {parent=hand_r, loc=<-1.7616 0.9753 -3.3814>, rot=<-0.0397 0.0814 -0.069 -0.9935>}
 pinky_01_r {parent=pinky_metacarpal_r, loc=<-7.1709 0.4831 0.1094>, rot=<-0.0654 -0.0035 0.2242 -0.9723>}
 pinky_02_r {parent=pinky_01_r, loc=<-2.8274 -0.0893 0.1043>, rot=<0.0037 0.0011 0.2854 -0.9584>}
 pinky_03_r {parent=pinky_02_r, loc=<-2.2046 -0.0307 0.0078>, rot=<-0.0002 0.0098 0.1445 -0.9895>}
 deform_wrist_dn_base_r {parent=hand_r, loc=<-0.0001 1.5328 0>, rot=<0 0 0 -1>}
 deform_wrist_dn_r {parent=deform_wrist_dn_base_r, loc=<0 0 0>, rot=<0 0 0 1>}
 deform_wrist_up_base_r {parent=hand_r, loc=<-0.0002 -2.3007 -0.0001>, rot=<0 0 0 -1>}
 deform_wrist_up_r {parent=deform_wrist_up_base_r, loc=<0 0 0>, rot=<0 0 0 1>}
 weapon_r {parent=hand_r, loc=<0.1974 3.4111 -0.3811>, rot=<0 0 0 1>}
 hand_attach_r {parent=hand_r, loc=<-8.5001 2.9999 0>, rot=<0 0 -0.7071 -0.7071>}
 lowerarm_twist_02_r {parent=lowerarm_r, loc=<-10.6707 0 0>, rot=<0 0 0 -1>}
 dyn_sleeve_dn_r {parent=lowerarm_twist_02_r, loc=<-1.0526 0.306 5.8963>, rot=<-0.0107 -0.9899 -0.0095 0.1408>}
 dyn_sleeve_bk_r {parent=lowerarm_twist_02_r, loc=<-1.364 5.8248 0.0655>, rot=<0.0438 -0.7952 -0.601 -0.0677>}
 dyn_sleeve_ft_r {parent=lowerarm_twist_02_r, loc=<-1.2419 -3.4821 -0.2785>, rot=<-0.044 -0.7133 0.6967 -0.0629>}
 dyn_sleeve_up_r {parent=lowerarm_twist_02_r, loc=<-0.8255 0.4133 -3.6915>, rot=<0 -1 0 0>}
 lowerarm_twist_01_r {parent=lowerarm_r, loc=<-18.316 0 0>, rot=<0 0 0 -1>}
 deform_elbow_in_base_r {parent=lowerarm_r, loc=<-0.1834 -3.6062 0.2572>, rot=<0 0 0 -1>}
 deform_elbow_in_r {parent=deform_elbow_in_base_r, loc=<0 0 0>, rot=<0 0 0 1>}
 deform_elbow_out_base_r {parent=lowerarm_r, loc=<0 4 -0.0001>, rot=<0 0 0 -1>}
 deform_elbow_out_r {parent=deform_elbow_out_base_r, loc=<0 0 0>, rot=<0 0 0 1>}
 upperarm_twist_01_r {parent=upperarm_r, loc=<-7.2135 0.1134 0>, rot=<0 0 0 -1>}
 deform_upperarm_bk_base_r {parent=upperarm_twist_01_r, loc=<6.4053 4.3746 -0.314>, rot=<0 0 0 -1>}
 deform_upperarm_bk_r {parent=deform_upperarm_bk_base_r, loc=<0 0 0>, rot=<0 0 0 1>}
 deform_upperarm_fr_base_r {parent=upperarm_twist_01_r, loc=<5.6312 -5.092 -0.3354>, rot=<0 0 0 -1>}
 deform_upperarm_fr_r {parent=deform_upperarm_fr_base_r, loc=<0 0 0>, rot=<0 0 0 1>}
 deform_upperarm_up_base_r {parent=upperarm_twist_01_r, loc=<8.891 -0.4825 -3.9517>, rot=<0 0 0 -1>}
 deform_upperarm_up_r {parent=deform_upperarm_up_base_r, loc=<0 0 0>, rot=<0 0 0 1>}
 upperarm_twist_02_r {parent=upperarm_r, loc=<-19.7998 0.1134 0>, rot=<0 0 0 -1>}
 base_armband_knot {parent=upperarm_twist_02_r, loc=<-2.2915 2.1731 -5.6654>, rot=<-0.0859 -0.7124 -0.0132 -0.6963>}
 dyn_armband_knot {parent=base_armband_knot, loc=<0 0 0>, rot=<0 0 0 1>}
 dyn_armband_low {parent=dyn_armband_knot, loc=<0.0922 0.7849 -1.2954>, rot=<0.7782 0.2126 -0.48 0.3447>}
 dyn_armband_hi {parent=dyn_armband_knot, loc=<-0.2798 -1.2257 0.6774>, rot=<0.9436 -0.2145 -0.0645 0.2439>}
 base_shoulders_r {parent=clavicle_r, loc=<-12.8912 -0.2319 -6.594>, rot=<-0.2206 -0.6855 -0.6597 0.215>}
 shoulders_r {parent=base_shoulders_r, loc=<0 0 0>, rot=<0 0 0 -1>}
 scarf_01_l {parent=spine_05, loc=<14.0266 1.449 -7.7234>, rot=<0.0284 -0.7065 -0.0284 -0.7065>}
 base_scarf_front_c {parent=spine_05, loc=<7.4339 8.1656 7.0562>, rot=<0.2646 0.6018 0.7238 0.2096>}
 scarf_front_c {parent=base_scarf_front_c, loc=<0 0 0>, rot=<0 0 0 1>}
 base_scarf_c {parent=spine_05, loc=<17.2592 -5.2031 0>, rot=<0.5196 -0.4795 0.4795 -0.5196>}
 scarf_c {parent=base_scarf_c, loc=<0 0 0>, rot=<0 0 0 1>}
 scarf_02_l {parent=spine_05, loc=<16.603 -4.055 -5.294>, rot=<0.3418 -0.619 0.291 -0.6444>}
 scarf_01_r {parent=spine_05, loc=<14.027 1.4491 7.7234>, rot=<0.7065 0.0284 0.7065 -0.0284>}
 scarf_02_r {parent=spine_05, loc=<16.6034 -4.0552 5.2936>, rot=<0.6444 -0.291 0.619 -0.3418>}
 deform_pec_aim_base_l {parent=spine_04, loc=<3.2456 9.3309 -1.2237>, rot=<0.0603 -0.4033 0.3318 -0.8507>}
 deform_pec_aim_l {parent=deform_pec_aim_base_l, loc=<0 0 0>, rot=<0 0 0 1>}
 deform_pec_base_l {parent=deform_pec_aim_l, loc=<9.3665 0 0>, rot=<0 0 0 1>}
 deform_pec_l {parent=deform_pec_base_l, loc=<0 0 0>, rot=<0 0 0 1>}
 deform_pec_aim_base_r {parent=spine_04, loc=<3.2452 9.3309 1.2237>, rot=<-0.4033 -0.0603 0.8507 0.3318>}
 deform_pec_aim_r {parent=deform_pec_aim_base_r, loc=<0 0 0>, rot=<0 0 0 1>}
 deform_pec_base_r {parent=deform_pec_aim_r, loc=<-9.3667 -0.0003 -0.0004>, rot=<0 0 0 1>}
 deform_pec_r {parent=deform_pec_base_r, loc=<0 0 0>, rot=<0 0 0 1>}
 Base_collar_L {parent=spine_04, loc=<5.4714 10.6782 -6.7505>, rot=<0.5883 -0.5827 -0.0773 -0.5553>}
 Dyn_collar_L {parent=Base_collar_L, loc=<0 0 0>, rot=<0 0 0 1>}
 Base_collar_R {parent=spine_04, loc=<5.4719 10.6782 6.7505>, rot=<0.5553 0.0773 0.5827 -0.5883>}
 Dyn_collar_R {parent=Base_collar_R, loc=<0 0 0>, rot=<0 0 0 1>}
 dyn_jacket_01 {parent=spine_04, loc=<0 0 0>, rot=<0 0 1 0>}
 dyn_jacket_01_l {parent=spine_04, loc=<0 0 0>, rot=<0 0 1 0>}
 dyn_jacket_01_r {parent=spine_04, loc=<0 0 0>, rot=<0 0 1 0>}
 dyn_knife_01_l_MSTR {parent=spine_01, loc=<2.4211 -8.2878 1.5696>, rot=<0.5089 -0.4367 0.6856 0.2834>}
 dyn_knife_01_l {parent=dyn_knife_01_l_MSTR, loc=<0 0 0>, rot=<0 0 0 1>}
 dyn_pouch_l_A_MSTR {parent=spine_01, loc=<3.3219 5.6104 -9.6658>, rot=<0.0286 -0.9376 -0.3326 -0.0972>}
 dyn_pouch_l_A {parent=dyn_pouch_l_A_MSTR, loc=<0 0 0>, rot=<0 0 0 1>}
 dyn_pouch_l_B_MSTR {parent=spine_01, loc=<4.5677 1.242 -10.9929>, rot=<0.0152 -0.9919 -0.0742 -0.1023>}
 dyn_pouch_l_B {parent=dyn_pouch_l_B_MSTR, loc=<0 0 0>, rot=<0 0 0 1>}
 thigh_l {parent=pelvis, loc=<-1.0855 -0.2109 -8.7053>, rot=<0.0009 0.0355 0.0155 -0.9993>}
 calf_l {parent=thigh_l, loc=<-42.5284 -0.0019 -0.7297>, rot=<0.0006 -0.0161 -0.0391 -0.9991>}
 foot_l {parent=calf_l, loc=<-46.7401 -2.5014 -0.592>, rot=<0.0427 -0.0238 0.0353 -0.9982>}
 ball_l {parent=foot_l, loc=<-3.5658 10.8066 -0.1544>, rot=<0.0001 -0.0001 -0.7063 -0.7079>}
 calf_twist_02_l {parent=calf_l, loc=<-19.1546 -1.7943 -0.2638>, rot=<0.0237 0.0095 -0.0035 -0.9997>}
 base_zipper_02_l {parent=calf_twist_02_l, loc=<-7.5245 4.1498 -4.6498>, rot=<-0.2935 -0.683 -0.2107 -0.6348>}
 zipper_02_l {parent=base_zipper_02_l, loc=<0 0 0>, rot=<0 0 0 1>}
 base_zipper_01_l {parent=calf_twist_02_l, loc=<-7.3311 3.7975 2.7484>, rot=<0.6912 0.2734 0.5987 0.2982>}
 zipper_01_l {parent=base_zipper_01_l, loc=<0 0 0>, rot=<0 0 0 1>}
 calf_twist_01_l {parent=calf_l, loc=<-35.8483 -1.9197 -0.5794>, rot=<0.0237 0.0095 -0.0035 -0.9997>}
 dyn_pants_sleeve_dn_l {parent=calf_twist_01_l, loc=<-0.6095 -0.5132 5.178>, rot=<-0.0445 -0.9982 -0.0257 0.0299>}
 dyn_pants_sleeve_bk_l {parent=calf_twist_01_l, loc=<-1.9827 -5.3372 -0.1478>, rot=<-0.0103 -0.724 0.6877 0.0526>}
 dyn_pants_sleeve_ft_l {parent=calf_twist_01_l, loc=<-0.9188 6.3185 0.0288>, rot=<0.0526 0.6877 0.724 0.0103>}
 dyn_pants_sleeve_up_l {parent=calf_twist_01_l, loc=<-0.941 -0.1303 -5.3836>, rot=<-0.0445 -0.9982 -0.0257 0.0299>}
 deform_knee_in_base_l {parent=calf_l, loc=<0.8031 -5.0231 0.2456>, rot=<0 0 0 -1>}
 deform_knee_in_l {parent=deform_knee_in_base_l, loc=<0 0 0>, rot=<0 0 0 1>}
 deform_knee_out_base_l {parent=calf_l, loc=<0 4 0>, rot=<0 0 0 -1>}
 deform_knee_out_l {parent=deform_knee_out_base_l, loc=<0 0 0>, rot=<0 0 0 1>}
 foot_l_reparent {parent=calf_l, loc=<-43.3242 -3.2165 -0.7056>, rot=<0.0427 -0.0238 0.0353 -0.9982>}
 ball_l_reparent {parent=foot_l_reparent, loc=<-9.312 9.5823 -0.1542>, rot=<0.0001 -0.0001 -0.7063 -0.7079>}
 thigh_twist_01_l {parent=thigh_l, loc=<-27.952 0 0>, rot=<0.0013 0 0 -1>}
 deform_groin_base_l {parent=thigh_l, loc=<0.4801 5.6776 -0.2486>, rot=<0 0 0 -1>}
 deform_groin_l {parent=deform_groin_base_l, loc=<0 0 0>, rot=<0 0 0 1>}
 correct_ALT_groin_L {parent=deform_groin_l, loc=<0 0 0>, rot=<0 0 0 1>}
 Base_skirt_bk_01_l {parent=thigh_l, loc=<-4.8118 -12.1715 3.4116>, rot=<-0.0372 -0.7977 0.6012 -0.0284>}
 dyn_skirt_bk_01_l {parent=Base_skirt_bk_01_l, loc=<0 0 0>, rot=<0 0 0 1>}
 Base_skirt_side_bk_01_l {parent=thigh_l, loc=<-3.8889 -5.7128 -5.7098>, rot=<-0.0209 -0.9627 0.2533 -0.0925>}
 dyn_skirt_side_bk_01_l {parent=Base_skirt_side_bk_01_l, loc=<0 0 0>, rot=<0 0 0 1>}
 Base_skirt_side_fr_01_l {parent=thigh_l, loc=<-4.1053 4.0552 -5.1689>, rot=<0.0175 -0.9311 -0.3461 -0.114>}
 dyn_skirt_side_fr_01_l {parent=Base_skirt_side_fr_01_l, loc=<0 0 0>, rot=<0 0 0 1>}
 Base_skirt_fr_01_l {parent=thigh_l, loc=<-2.31 9.1612 3.4826>, rot=<0.0208 -0.7693 -0.6315 -0.0945>}
 dyn_skirt_fr_01_l {parent=Base_skirt_fr_01_l, loc=<0 0 0>, rot=<0 0 0 1>}
 dyn_skirt_fr_02_l {parent=dyn_skirt_fr_01_l, loc=<5.2389 0 0>, rot=<0 0 0 1>}
 thigh_Correct_ALT_L {parent=thigh_l, loc=<-19.1151 0 0>, rot=<0.0013 0 0 -1>}
 base_knee_l {parent=thigh_l, loc=<-39.0484 0 0>, rot=<-0.0109 -0.0118 -0.7341 -0.6789>}
 knee_l {parent=base_knee_l, loc=<0 0 0>, rot=<0 0 0 1>}
 thigh_r {parent=pelvis, loc=<-1.0855 -0.2109 8.7053>, rot=<0.0355 -0.0009 0.9993 0.0155>}
 calf_r {parent=thigh_r, loc=<42.5284 0.0019 0.7297>, rot=<0.0006 -0.0161 -0.0391 -0.9991>}
 foot_r {parent=calf_r, loc=<46.7401 2.5014 0.592>, rot=<0.0427 -0.0238 0.0353 -0.9982>}
 ball_r {parent=foot_r, loc=<3.5658 -10.8066 0.1544>, rot=<0.0001 -0.0001 -0.7063 -0.7079>}
 calf_twist_02_r {parent=calf_r, loc=<19.1546 1.7943 0.2638>, rot=<0.0237 0.0095 -0.0035 -0.9997>}
 base_zipper_02_r {parent=calf_twist_02_r, loc=<7.5245 -4.1498 4.6498>, rot=<-0.2107 0.6348 0.2935 -0.683>}
 zipper_02_r {parent=base_zipper_02_r, loc=<0 0 0>, rot=<0 0 0 -1>}
 base_zipper_01_r {parent=calf_twist_02_r, loc=<7.3312 -3.7975 -2.7485>, rot=<-0.5987 0.2982 0.6912 -0.2734>}
 zipper_01_r {parent=base_zipper_01_r, loc=<0 0 0>, rot=<0 0 0 1>}
 calf_twist_01_r {parent=calf_r, loc=<35.8483 1.9197 0.5794>, rot=<0.0237 0.0095 -0.0035 -0.9997>}
 dyn_pants_sleeve_dn_r {parent=calf_twist_01_r, loc=<0.7768 0.4923 -5.2174>, rot=<-0.0257 -0.0299 0.0445 -0.9982>}
 dyn_pants_sleeve_bk_r {parent=calf_twist_01_r, loc=<1.9185 6.1525 0.1331>, rot=<0.6877 -0.0526 0.0103 -0.724>}
 dyn_pants_sleeve_ft_r {parent=calf_twist_01_r, loc=<0.6559 -6.6353 -0.1155>, rot=<-0.724 0.0103 0.0526 -0.6877>}
 dyn_pants_sleeve_up_r {parent=calf_twist_01_r, loc=<1.1746 0.041 5.4328>, rot=<-0.0257 -0.0299 0.0445 -0.9982>}
 deform_knee_in_base_r {parent=calf_r, loc=<-0.8031 5.0231 -0.2457>, rot=<0 0 0 -1>}
 deform_knee_in_r {parent=deform_knee_in_base_r, loc=<0 0 0>, rot=<0 0 0 1>}
 deform_knee_out_base_r {parent=calf_r, loc=<0 -4 0>, rot=<0 0 0 -1>}
 deform_knee_out_r {parent=deform_knee_out_base_r, loc=<0 0 0>, rot=<0 0 0 1>}
 foot_r_reparent {parent=calf_r, loc=<43.3242 3.2165 0.7056>, rot=<0.0427 -0.0238 0.0353 -0.9982>}
 ball_r_reparent {parent=foot_r_reparent, loc=<9.312 -9.5823 0.1542>, rot=<0.0001 -0.0001 -0.7063 -0.7079>}
 thigh_twist_01_r {parent=thigh_r, loc=<27.952 0 0>, rot=<0.0013 0 0 -1>}
 dyn_gun_MSTR {parent=thigh_twist_01_r, loc=<-7.6497 -0.2376 7.7404>, rot=<-0.0184 -0.0189 -0.0228 -0.9994>}
 dyn_gun {parent=dyn_gun_MSTR, loc=<0 0 0>, rot=<0 0 0 1>}
 deform_groin_base_r {parent=thigh_r, loc=<-0.4801 -5.6776 0.2486>, rot=<0 0 0 -1>}
 deform_groin_r {parent=deform_groin_base_r, loc=<0 0 0>, rot=<0 0 0 1>}
 correct_ALT_groin_R {parent=deform_groin_r, loc=<0 0 0>, rot=<0 0 0 1>}
 Base_skirt_bk_01_r {parent=thigh_r, loc=<4.8117 12.1716 -3.4115>, rot=<0.6012 0.0284 0.0372 -0.7977>}
 dyn_skirt_bk_01_r {parent=Base_skirt_bk_01_r, loc=<0 0 0>, rot=<0 0 0 1>}
 Base_skirt_side_bk_01_r {parent=thigh_r, loc=<3.8889 5.7128 5.7098>, rot=<0.2533 0.0925 0.0209 -0.9627>}
 dyn_skirt_side_bk_01_r {parent=Base_skirt_side_bk_01_r, loc=<0 0 0>, rot=<0 0 0 1>}
 Base_skirt_side_fr_01_r {parent=thigh_r, loc=<4.1054 -4.0552 5.1688>, rot=<-0.3461 0.114 -0.0175 -0.9311>}
 dyn_skirt_side_fr_01_r {parent=Base_skirt_side_fr_01_r, loc=<0 0 0>, rot=<0 0 0 1>}
 Base_skirt_fr_01_r {parent=thigh_r, loc=<2.31 -9.1612 -3.4826>, rot=<-0.6315 0.0945 -0.0208 -0.7693>}
 dyn_skirt_fr_01_r {parent=Base_skirt_fr_01_r, loc=<0 0 0>, rot=<0 0 0 1>}
 dyn_skirt_fr_02_r {parent=dyn_skirt_fr_01_r, loc=<5.2389 0 0>, rot=<0 0 0 1>}
 thigh_Correct_ALT_R {parent=thigh_r, loc=<19.1255 0 0>, rot=<0.0013 0 0 -1>}
 base_knee_r {parent=thigh_r, loc=<39.0484 0 0>, rot=<0.0118 -0.0109 0.6789 -0.7341>}
 knee_r {parent=base_knee_r, loc=<0 0 0>, rot=<0 0 0 -1>}
 deform_glute_base_l {parent=pelvis, loc=<-7.3466 -9.0213 -9.1245>, rot=<0.0172 0.031 -0.4477 -0.8935>}
 deform_glute_l {parent=deform_glute_base_l, loc=<0 0 0>, rot=<0 0 0 1>}
 deform_glute_base_r {parent=pelvis, loc=<-7.3464 -9.0213 9.1245>, rot=<0.031 -0.0172 0.8935 -0.4477>}
 deform_glute_r {parent=deform_glute_base_r, loc=<0 0 0>, rot=<0 0 0 1>}
 dyn_belt_01_r_MSTR {parent=pelvis, loc=<6.0053 8.7743 2.8041>, rot=<-0.3392 -0.704 -0.5826 0.2233>}
 dyn_belt_01_r {parent=dyn_belt_01_r_MSTR, loc=<0 0 0>, rot=<0 0 0 1>}
 dyn_belt_02_r {parent=dyn_belt_01_r, loc=<5.3564 0 0>, rot=<-0.3584 0.1176 0.2687 -0.8863>}
 dyn_belt_03_r {parent=dyn_belt_02_r, loc=<5.5813 0 0>, rot=<-0.1125 0.0105 0.0156 -0.9935>}
 dyn_belt_04_r {parent=dyn_belt_03_r, loc=<5.0716 0 0>, rot=<-0.0123 -0.0003 0.0234 -0.9997>}
 Correct_ALT_belt_C {parent=pelvis, loc=<11.6036 0.1145 0>, rot=<0 0 0.0179 -0.9998>}
 base_pouch_l {parent=pelvis, loc=<7.4253 -0.0238 -14.5873>, rot=<-0.0705 -0.6828 -0.0242 -0.7268>}
 pouch_l {parent=base_pouch_l, loc=<0 0 0>, rot=<0 0 0 1>}
 base_pouch_02_r {parent=pelvis, loc=<7.2886 1.6063 14.2086>, rot=<0.7006 0.0673 0.7018 0.1104>}
 base_pouch_03_r {parent=pelvis, loc=<7.2724 4.575 12.9538>, rot=<0.6703 0.218 0.6585 0.2638>}
 base_pouch_01_r {parent=pelvis, loc=<7.313 -1.6887 14.4115>, rot=<0.6975 -0.0871 0.7096 -0.0493>}
 attach {parent=root, loc=<0 0 0>, rot=<0 0 0 1>}
 ik_hand_root {parent=root, loc=<0 0 0>, rot=<0 0 0 1>}
 ik_hand_gun {parent=ik_hand_root, loc=<-48.1743 3.3545 104.4381>, rot=<0.6276 0.1309 0.4183 -0.6434>}
 ik_hand_r {parent=ik_hand_gun, loc=<0 0 0>, rot=<0 0 0 1>}
 ik_hand_l {parent=ik_hand_gun, loc=<59.3271 67.6997 34.3524>, rot=<-0.6158 -0.7027 -0.3565 0>}
 ik_foot_root {parent=root, loc=<0 0 0>, rot=<0 0 0 1>}
 ik_foot_l {parent=ik_foot_root, loc=<14.8262 -7.3545 8.296>, rot=<0.0357 0.7045 0.0284 -0.7082>}
 ik_foot_r {parent=ik_foot_root, loc=<-14.8262 -7.3545 8.296>, rot=<0.7082 0.0284 -0.7045 0.0357>}
 root_fp {parent=root, loc=<0 0 0>, rot=<0 0 0 1>}
]
variables [
 Backward_Movement {type=double, default=0}
 Distance calf {type=double, default=0}
 Emotes {type=bool, default=False}
 FB_Movement {type=double, default=0}
 Forward_Movement {type=double, default=0}
 High Fps {type=double, default=0}
 LR_Movement {type=double, default=0}
 Left_Movement {type=double, default=0}
 Lod1 {type=bool, default=False}
 Lod2 {type=bool, default=False}
 NoStand {type=double, default=0}
 PlayingEmote {type=bool, default=False}
 RdanOFF {type=bool, default=False}
 Right_Movement {type=double, default=0}
 Sprint and LowFPS {type=bool, default=False}
 Stand {type=double, default=0}
 TacticalSprinting {type=bool, default=False}
 crouch {type=double, default=0}
 sit_lower {type=double, default=0}
]
