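# Physics asset for the BlowWire medium female body's belt, covering a right-side belt chain and the surrounding pelvis, thigh and skirt bones with capsules.
# The belt runs from a master body on the pelvis through four links, with swing and twist limits that tighten toward the tip.
# Front skirt chains on both thighs and a right-only side skirt body use angularly free constraints, as do the thighs and belt master.
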PhysicsAsset /BRCosmetics/Characters/Player/Female/Medium/Bodies/F_MED_BlowWire/Meshes/F_MED_BlowWire_Physics_belt
bodies {
 Base_skirt_fr_01_l [{shapeType=Capsule, length=5.1577, radius=4.0779, shapeName=Base_skirt_fr_01_l_capsule, center=<0 -1.7442 -2.9094>, rotation=<rot 0 0 90>}]
 Base_skirt_fr_01_r [{shapeType=Capsule, length=6.1577, radius=4.0779, shapeName=dyn_skirt_fr_01_r_capsule, center=<0.4116 1.8933 -3.197>, rotation=<rot 0 0 90>}]
 dyn_belt_01_r [{shapeType=Capsule, length=2, radius=2, shapeName=dyn_belt_01_r_capsule, rotation=<rot 80 0 0>}]
 dyn_belt_01_r_MSTR [{shapeType=Capsule, length=0.505, radius=0.505, shapeName=dyn_belt_01_r_MSTR_capsule}]
 dyn_belt_02_r [{shapeType=Capsule, length=2, radius=2, shapeName=dyn_belt_02_r_capsule, rotation=<rot 80 0 0>}]
 dyn_belt_03_r [{shapeType=Capsule, length=2, radius=2, shapeName=dyn_belt_03_r_capsule, rotation=<rot 80 0 0>}]
 dyn_belt_04_r [{shapeType=Capsule, length=2, radius=2, shapeName=dyn_belt_04_r_capsule, rotation=<rot 80 0 0>}]
 dyn_skirt_fr_01_l [{shapeType=Capsule, length=5.1577, radius=4.0779, shapeName=dyn_skirt_fr_01_l_capsule, center=<2.6037 -1.6273 -2.9094>, rotation=<rot 0 0 90>}]
 dyn_skirt_fr_01_r [{shapeType=Capsule, length=6.1577, radius=4.0779, shapeName=dyn_skirt_fr_01_r_capsule, center=<3.1354 1.47 -3.197>, rotation=<rot 0 0 90>}]
 dyn_skirt_fr_02_l [{shapeType=Capsule, length=5.1577, radius=4.0779, shapeName=dyn_skirt_fr_02_l_capsule, center=<2.6037 -1.3728 -2.9094>, rotation=<rot 0 0 90>}]
 dyn_skirt_fr_02_r [{shapeType=Capsule, length=6.1577, radius=4.0779, shapeName=dyn_skirt_fr_02_r_capsule, center=<3.1354 1.1461 -3.197>, rotation=<rot 0 0 90>}]
 dyn_skirt_side_fr_01_r [{shapeType=Capsule, length=4.5811, radius=3.4189, shapeName=dyn_skirt_side_fr_01_r_capsule, center=<3.4846 0 -2.6608>, rotation=<rot 90 0 0>}]
 pelvis [{shapeType=Capsule, length=9.6324, radius=8.6723, shapeName=pelvis_capsule}]
 thigh_l [{shapeType=Capsule, length=15.2434, radius=9.6818, shapeName=thigh_r_capsule, center=<-15.6854 0 0>, rotation=<rot 90 0 180>}]
 thigh_r [{shapeType=Capsule, length=21.5268, radius=8.2935, shapeName=thigh_r_capsule, center=<21.8563 0 0>, rotation=<rot 90 0 0>}]
}
constraints [
 {
  bone1Name=dyn_skirt_fr_01_l
  bone2Name=Base_skirt_fr_01_l
  swing1LimitDegrees=45
  swing1Motion=Free
  swing2LimitDegrees=45
  swing2Motion=Free
  twistLimitDegrees=45
  twistMotion=Free
 }
 {
  bone1Name=dyn_skirt_fr_02_l
  bone2Name=dyn_skirt_fr_01_l
  swing1LimitDegrees=45
  swing1Motion=Free
  swing2LimitDegrees=45
  swing2Motion=Free
  twistLimitDegrees=45
  twistMotion=Free
 }
 {
  bone1Name=Base_skirt_fr_01_l
  bone2Name=thigh_l
  swing1LimitDegrees=45
  swing1Motion=Free
  swing2LimitDegrees=45
  swing2Motion=Free
  twistLimitDegrees=45
  twistMotion=Free
 }
 {
  bone1Name=thigh_r
  bone2Name=pelvis
  swing1LimitDegrees=45
  swing1Motion=Free
  swing2LimitDegrees=45
  swing2Motion=Free
  twistLimitDegrees=45
  twistMotion=Free
 }
 {
  bone1Name=dyn_belt_01_r_MSTR
  bone2Name=pelvis
  swing1LimitDegrees=45
  swing1Motion=Free
  swing2LimitDegrees=45
  swing2Motion=Free
  twistLimitDegrees=45
  twistMotion=Free
 }
 {bone1Name=dyn_belt_01_r, bone2Name=dyn_belt_01_r_MSTR, swing1LimitDegrees=40, swing2LimitDegrees=40, twistLimitDegrees=15}
 {bone1Name=dyn_belt_02_r, bone2Name=dyn_belt_01_r, swing1LimitDegrees=25, swing2LimitDegrees=25, twistLimitDegrees=15}
 {bone1Name=dyn_belt_03_r, bone2Name=dyn_belt_02_r, swing1LimitDegrees=20, swing2LimitDegrees=20, twistLimitDegrees=10}
 {bone1Name=dyn_belt_04_r, bone2Name=dyn_belt_03_r, swing1LimitDegrees=15, swing2LimitDegrees=15, twistLimitDegrees=5}
 {
  bone1Name=thigh_l
  bone2Name=pelvis
  swing1LimitDegrees=45
  swing1Motion=Free
  swing2LimitDegrees=45
  swing2Motion=Free
  twistLimitDegrees=45
  twistMotion=Free
 }
 {
  bone1Name=dyn_skirt_side_fr_01_r
  bone2Name=thigh_r
  swing1LimitDegrees=45
  swing1Motion=Free
  swing2LimitDegrees=45
  swing2Motion=Free
  twistLimitDegrees=45
  twistMotion=Free
 }
 {
  bone1Name=dyn_skirt_fr_02_r
  bone2Name=dyn_skirt_fr_01_r
  swing1LimitDegrees=45
  swing1Motion=Free
  swing2LimitDegrees=45
  swing2Motion=Free
  twistLimitDegrees=45
  twistMotion=Free
 }
 {
  bone1Name=dyn_skirt_fr_01_r
  bone2Name=Base_skirt_fr_01_r
  swing1LimitDegrees=45
  swing1Motion=Free
  swing2LimitDegrees=45
  swing2Motion=Free
  twistLimitDegrees=45
  twistMotion=Free
 }
 {
  bone1Name=Base_skirt_fr_01_r
  bone2Name=thigh_r
  swing1LimitDegrees=45
  swing1Motion=Free
  swing2LimitDegrees=45
  swing2Motion=Free
  twistLimitDegrees=45
  twistMotion=Free
 }
]
