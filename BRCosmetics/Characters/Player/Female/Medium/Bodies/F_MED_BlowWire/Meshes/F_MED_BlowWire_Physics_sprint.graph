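# Physics asset for the skirt of the medium female BlowWire cosmetic body, apparently a sprint variant by its name.
# Capsules on pelvis and thighs form the leg volume, while tiny Base_skirt bodies anchor larger dyn_skirt capsules, two chained at each front.
# Limited swing and twist constraints shape each panel's motion.
# Angularly free constraints link neighboring panels per side and most anchors to thighs, except the left side pair.
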PhysicsAsset /BRCosmetics/Characters/Player/Female/Medium/Bodies/F_MED_BlowWire/Meshes/F_MED_BlowWire_Physics_sprint
bodies {
 Base_skirt_bk_01_l [{shapeType=Capsule, length=0.505, radius=0.505, shapeName=Base_skirt_bk_01_l_capsule, rotation=<rot 0 -90 0>}]
 Base_skirt_bk_01_r [{shapeType=Capsule, length=0.505, radius=0.505, shapeName=Base_skirt_bk_01_r_capsule, rotation=<rot 0 -90 0>}]
 Base_skirt_fr_01_l [{shapeType=Capsule, length=0.5, radius=0.5, shapeName=Base_skirt_fr_01_l_capsule, rotation=<rot 0 0 90>}]
 Base_skirt_fr_01_r [{shapeType=Capsule, length=0.5, radius=0.5, shapeName=Base_skirt_fr_01_r_capsule, rotation=<rot 0 0 90>}]
 Base_skirt_side_bk_01_l [{shapeType=Capsule, length=0.505, radius=0.505, shapeName=Base_skirt_side_bk_01_l_capsule}]
 Base_skirt_side_bk_01_r [{shapeType=Capsule, length=0.505, radius=0.505, shapeName=Base_skirt_side_bk_01_r_capsule}]
 Base_skirt_side_fr_01_l [{shapeType=Capsule, length=0.505, radius=0.505, shapeName=Base_skirt_side_fr_01_l_capsule}]
 Base_skirt_side_fr_01_r [{shapeType=Capsule, length=0.505, radius=0.505, shapeName=Base_skirt_side_fr_01_r_capsule}]
 dyn_skirt_bk_01_l [{shapeType=Capsule, length=6.1596, radius=4.0607, shapeName=dyn_skirt_bk_01_l_capsule, center=<6.3799 0 3.1078>, rotation=<rot 0 0 90>}]
 dyn_skirt_bk_01_r [{shapeType=Capsule, length=6.2609, radius=4.0325, shapeName=dyn_skirt_bk_01_r_capsule, center=<6.3799 0 3.1078>, rotation=<rot 0 0 90>}]
 dyn_skirt_fr_01_l [{shapeType=Capsule, length=5, radius=4, shapeName=dyn_skirt_fr_01_l_capsule, center=<4.0037 0 1.1986>, rotation=<rot 0 0 90>}]
 dyn_skirt_fr_01_r [{shapeType=Capsule, length=5, radius=4, shapeName=dyn_skirt_fr_01_r_capsule, center=<4.0037 0 1.1986>, rotation=<rot 0 0 90>}]
 dyn_skirt_fr_02_l [{shapeType=Capsule, length=5.886, radius=4, shapeName=dyn_skirt_fr_02_l_capsule, center=<4.0037 0 3.3875>, rotation=<rot 0 0 90>}]
 dyn_skirt_fr_02_r [{shapeType=Capsule, length=5.886, radius=4, shapeName=dyn_skirt_fr_02_r_capsule, center=<4.0037 0 3.3875>, rotation=<rot 0 0 90>}]
 dyn_skirt_side_bk_01_l [{shapeType=Capsule, length=5.4343, radius=5.0849, shapeName=dyn_skirt_side_bk_01_l_capsule, center=<4.5956 0 3.9926>, rotation=<rot 0 0 90>}]
 dyn_skirt_side_bk_01_r [{shapeType=Capsule, length=5.8162, radius=4.703, shapeName=dyn_skirt_side_bk_01_r_capsule, center=<4.6676 0 3.8581>, rotation=<rot 0 0 90>}]
 dyn_skirt_side_fr_01_l [{
   shapeType=Capsule
   length=7.5052
   radius=4
   shapeName=dyn_skirt_side_fr_01_l_capsule
   center=<3.6615 -0.0393 3.1855>
   rotation=<rot 0.0001 0.0001 85>
  }]
 dyn_skirt_side_fr_01_r [{shapeType=Capsule, length=6.814, radius=4.6912, shapeName=dyn_skirt_side_fr_01_r_capsule, center=<4.5347 0 3.7363>, rotation=<rot 0 0 90>}]
 pelvis [
  {shapeType=Capsule, length=3.8912, radius=11.1766, shapeName=pelvis_capsule}
  {shapeType=Capsule, length=9.2441, radius=8.6742, shapeName=pelvis_capsule, center=<-7.8203 -1.5176 0>}
 ]
 thigh_l [{shapeType=Capsule, length=16.2206, radius=8.4312, shapeName=thigh_l_capsule, center=<-18.3959 0 0>, rotation=<rot 90 0 0>}]
 thigh_r [{shapeType=Capsule, length=16.6636, radius=7.9881, shapeName=thigh_l_capsule, center=<18.3959 0 0>, rotation=<rot 90 0 180>}]
}
constraints [
 {bone1Name=dyn_skirt_side_bk_01_l, bone2Name=Base_skirt_side_bk_01_l, swing1LimitDegrees=45, swing2LimitDegrees=45, twistLimitDegrees=10}
 {bone1Name=dyn_skirt_side_fr_01_l, bone2Name=Base_skirt_side_fr_01_l, swing1LimitDegrees=45, swing2LimitDegrees=45, twistLimitDegrees=10}
 {
  bone1Name=Base_skirt_side_bk_01_r
  bone2Name=thigh_r
  swing1LimitDegrees=45
  swing1Motion=Free
  swing2LimitDegrees=45
  swing2Motion=Free
  twistLimitDegrees=45
  twistMotion=Free
 }
 {bone1Name=dyn_skirt_side_bk_01_r, bone2Name=Base_skirt_side_bk_01_r, swing1LimitDegrees=45, swing2LimitDegrees=45, twistLimitDegrees=10}
 {
  bone1Name=Base_skirt_side_fr_01_r
  bone2Name=thigh_r
  swing1LimitDegrees=45
  swing1Motion=Free
  swing2LimitDegrees=45
  swing2Motion=Free
  twistLimitDegrees=45
  twistMotion=Free
 }
 {bone1Name=dyn_skirt_side_fr_01_r, bone2Name=Base_skirt_side_fr_01_r, swing1LimitDegrees=45, swing2LimitDegrees=45, twistLimitDegrees=10}
 {
  bone1Name=thigh_l
  bone2Name=pelvis
  swing1LimitDegrees=45
  swing1Motion=Free
  swing2LimitDegrees=45
  swing2Motion=Free
  twistLimitDegrees=45
  twistMotion=Free
 }
 {bone1Name=dyn_skirt_fr_02_r, bone2Name=dyn_skirt_fr_01_r, swing1LimitDegrees=15, swing2LimitDegrees=15, twistLimitDegrees=0}
 {bone1Name=dyn_skirt_fr_01_r, bone2Name=Base_skirt_fr_01_r, swing1LimitDegrees=25, swing2LimitDegrees=25, twistLimitDegrees=0}
 {
  bone1Name=Base_skirt_fr_01_r
  bone2Name=thigh_r
  swing1LimitDegrees=45
  swing1Motion=Free
  swing2LimitDegrees=45
  swing2Motion=Free
  twistLimitDegrees=45
  twistMotion=Free
 }
 {bone1Name=dyn_skirt_fr_02_l, bone2Name=dyn_skirt_fr_01_l, swing1LimitDegrees=15, swing2LimitDegrees=15, twistLimitDegrees=0}
 {bone1Name=dyn_skirt_fr_01_l, bone2Name=Base_skirt_fr_01_l, swing1LimitDegrees=25, swing2LimitDegrees=20, twistLimitDegrees=0}
 {
  bone1Name=Base_skirt_fr_01_l
  bone2Name=thigh_l
  swing1LimitDegrees=45
  swing1Motion=Free
  swing2LimitDegrees=45
  swing2Motion=Free
  twistLimitDegrees=45
  twistMotion=Free
 }
 {
  bone1Name=dyn_skirt_side_fr_01_l
  bone2Name=dyn_skirt_side_bk_01_l
  swing1LimitDegrees=45
  swing1Motion=Free
  swing2LimitDegrees=45
  swing2Motion=Free
  twistLimitDegrees=45
  twistMotion=Free
 }
 {
  bone1Name=dyn_skirt_side_fr_01_r
  bone2Name=dyn_skirt_side_bk_01_r
  swing1LimitDegrees=45
  swing1Motion=Free
  swing2LimitDegrees=45
  swing2Motion=Free
  twistLimitDegrees=45
  twistMotion=Free
 }
 {bone1Name=dyn_skirt_bk_01_l, bone2Name=Base_skirt_bk_01_l, swing1LimitDegrees=20, swing2LimitDegrees=20, twistLimitDegrees=10}
 {
  bone1Name=Base_skirt_bk_01_l
  bone2Name=thigh_l
  swing1LimitDegrees=45
  swing1Motion=Free
  swing2LimitDegrees=45
  swing2Motion=Free
  twistLimitDegrees=45
  twistMotion=Free
 }
 {bone1Name=dyn_skirt_bk_01_r, bone2Name=Base_skirt_bk_01_r, swing1LimitDegrees=20, swing2LimitDegrees=20, twistLimitDegrees=10}
 {
  bone1Name=Base_skirt_bk_01_r
  bone2Name=thigh_r
  swing1LimitDegrees=45
  swing1Motion=Free
  swing2LimitDegrees=45
  swing2Motion=Free
  twistLimitDegrees=45
  twistMotion=Free
 }
 {
  bone1Name=dyn_skirt_side_bk_01_l
  bone2Name=dyn_skirt_bk_01_l
  swing1LimitDegrees=45
  swing1Motion=Free
  swing2LimitDegrees=45
  swing2Motion=Free
  twistLimitDegrees=45
  twistMotion=Free
 }
 {
  bone1Name=dyn_skirt_side_bk_01_r
  bone2Name=dyn_skirt_bk_01_r
  swing1LimitDegrees=45
  swing1Motion=Free
  swing2LimitDegrees=45
  swing2Motion=Free
  twistLimitDegrees=45
  twistMotion=Free
 }
 {
  bone1Name=thigh_r
  bone2Name=pelvis
  swing1LimitDegrees=45
  swing1Motion=Free
  swing2LimitDegrees=45
  swing2Motion=Free
  twistLimitDegrees=45
  twistMotion=Free
 }
 {
  bone1Name=dyn_skirt_fr_02_l
  bone2Name=dyn_skirt_side_fr_01_l
  swing1LimitDegrees=45
  swing1Motion=Free
  swing2LimitDegrees=45
  swing2Motion=Free
  twistLimitDegrees=45
  twistMotion=Free
 }
 {
  bone1Name=dyn_skirt_fr_02_r
  bone2Name=dyn_skirt_side_fr_01_r
  swing1LimitDegrees=45
  swing1Motion=Free
  swing2LimitDegrees=45
  swing2Motion=Free
  twistLimitDegrees=45
  twistMotion=Free
 }
]
