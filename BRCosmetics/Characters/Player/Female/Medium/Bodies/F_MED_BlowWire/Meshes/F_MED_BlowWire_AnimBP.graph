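# Post-process style body AnimBP for the F_MED_BlowWire cosmetic, adding secondary motion to skirt, belt, pouches, knife and collar.
# The parent mesh's pose runs through a Control Rig fed by movement, sprint, emote, LOD and FPS values from the event graph.
# AnimDynamics chains and a belt RigidBody follow, with toggleable skirt physics and thigh spherical limits at the front.
# A base/sprint RigidBody blend is left unconnected to the output.
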AnimBlueprint /BRCosmetics/Characters/Player/Female/Medium/Bodies/F_MED_BlowWire/Meshes/F_MED_BlowWire_AnimBP
target_skeleton=&/F_MED_BlowWire_Skeleton
graphs [
 AnimGraph {nodes [
   n1 = RigidBody!(ComponentPose=Local To Component(LocalPose=Use cached pose 'Base'()), Alpha, OverrideWorldGravity:Vector="0,0,-980", ExternalForce) -> (Pose) {
    bClampLinearTranslationLimitToRefPose=true
    bOverrideWorldGravity=true
    overridePhysicsAsset=&/F_MED_BlowWire_Physics_Base
    simulationSpace=BaseBoneSpace
    alphaBoolBlend {blendOption=Linear}
    baseBoneRef {boneName=pelvis}
    componentAppliedLinearAccClamp=<5000 5000 5000>
    componentLinearAccScale=<0.6 0.6 0.6>
    componentLinearVelScale=<0.6 0.6 0.6>
    overrideWorldGravity {z=-980}
    simSpaceSettings {maxAngularAcceleration=500, maxAngularVelocity=3, maxLinearAcceleration=400, maxLinearVelocity=250, velocityScaleZ=0, worldAlpha=1}
   }
   n2 = RigidBody(ComponentPose=Local To Component(), OverrideWorldGravity="0,0,-980") {
    bClampLinearTranslationLimitToRefPose=true
    bOverrideWorldGravity=true
    overridePhysicsAsset=&/F_MED_BlowWire_Physics_sprint
    simulationSpace=BaseBoneSpace
    alphaBoolBlend {blendOption=Linear}
    baseBoneRef {boneName=pelvis}
    componentAppliedLinearAccClamp=<5000 5000 5000>
    componentLinearAccScale=<0.6 0.6 0.6>
    componentLinearVelScale=<0.6 0.6 0.6>
    overrideWorldGravity {z=-980}
    simSpaceSettings {maxAngularAcceleration=500, maxAngularVelocity=3, maxLinearAcceleration=400, maxLinearVelocity=250, velocityScaleZ=0, worldAlpha=1}
   }
   Local To Component(LocalPose=Blend Poses by bool(
     BlendPose_0=Component To Local(ComponentPose=@n2)
     BlendPose_1=Component To Local(ComponentPose=@n1)
     BlendTime_0=0
     BlendTime_1=1
     bActiveValue=Get Sprint and LowFPS() -> (Sprint and LowFPS:bool)
    ) {blendType=Linear})
   n3 = Get bRigidBodyEnabled()
   n4 = Anim Dynamics - Bone: dyn_skirt_fr_01_l!(ComponentPose=Local To Component(LocalPose=Use cached pose 'Base'()), bAlphaBoolEnabled:bool=@n3, ExternalForce="0,0,0") -> (Pose) {
    alphaInputType=Bool
    angularDampingOverride=0.999
    angularSpringConstant=300
    bAngularSpring=true
    bLinearSpring=true
    bOverrideAngularDamping=true
    bOverrideLinearDamping=true
    bUseGravityOverride=true
    bUseSphericalLimits=true
    linearDampingOverride=0.999
    linearSpringConstant=50
    simulationSpace=RootRelative
    alphaBoolBlend {blendOption=Linear}
    boundBone {boneName=dyn_skirt_fr_01_l}
    componentAppliedLinearAccClamp=<7000 7000 7000>
    componentLinearAccScale=<8 8 8>
    componentLinearVelScale=<8 8 8>
    gravityOverride {z=-980}
    physicsBodyDefinitions [{
      collisionType=CustomSphere
      sphereCollisionRadius=4
      boundBone {boneName=dyn_skirt_fr_01_l}
      boxExtents=<10 10 10>
      constraintSetup {angularLimitsMax=<3 0 7>, angularLimitsMin=<-3 -25 -7>, angularTarget=<1 0 0>, linearAxesMax=<0 0 1>}
      localJointOffset=<0.5 0 0>
     }]
    relativeSpaceBone {boneName=pelvis}
    sphericalLimits [
     {limitRadius=4.5, drivingBone {boneName=thigh_l}, sphereLocalOffset=<-5 0 0>}
     {limitRadius=3.5, drivingBone {boneName=thigh_l}, sphereLocalOffset=<-5 0 5>}
    ]
   }
   n5 = Anim Dynamics - Bone: dyn_skirt_fr_02_l!(ComponentPose=@n4, bAlphaBoolEnabled:bool=@n3, ExternalForce="0,0,0") -> (Pose) {
    alphaInputType=Bool
    angularDampingOverride=0.999
    angularSpringConstant=300
    bAngularSpring=true
    bLinearSpring=true
    bOverrideAngularDamping=true
    bOverrideLinearDamping=true
    bUseGravityOverride=true
    bUseSphericalLimits=true
    linearDampingOverride=0.999
    linearSpringConstant=50
    simulationSpace=RootRelative
    alphaBoolBlend {blendOption=Linear}
    boundBone {boneName=dyn_skirt_fr_02_l}
    componentAppliedLinearAccClamp=<7000 7000 7000>
    componentLinearAccScale=<8 8 8>
    componentLinearVelScale=<8 8 8>
    gravityOverride {z=-980}
    physicsBodyDefinitions [{
      collisionType=CustomSphere
      sphereCollisionRadius=4
      boundBone {boneName=dyn_skirt_fr_02_l}
      boxExtents=<10 10 10>
      constraintSetup {angularLimitsMax=<3 0 7>, angularLimitsMin=<-3 -25 -7>, angularTarget=<1 0 0>, linearAxesMax=<0 0 1>}
      localJointOffset=<0.5 0 0>
     }]
    relativeSpaceBone {boneName=pelvis}
    sphericalLimits [
     {limitRadius=4.5, drivingBone {boneName=thigh_l}, sphereLocalOffset=<-10 0 0>}
     {limitRadius=4, drivingBone {boneName=thigh_l}, sphereLocalOffset=<-10 0 5>}
    ]
   }
   n6 = Anim Dynamics - Bone: dyn_skirt_side_fr_01_l!(ComponentPose=@n5, bAlphaBoolEnabled:bool=@n3, ExternalForce="0,0,0") -> (Pose) {
    alphaInputType=Bool
    angularDampingOverride=0.999
    angularSpringConstant=300
    bAngularSpring=true
    bLinearSpring=true
    bOverrideAngularDamping=true
    bOverrideLinearDamping=true
    bUseGravityOverride=true
    bUseSphericalLimits=true
    linearDampingOverride=0.999
    linearSpringConstant=50
    simulationSpace=RootRelative
    alphaBoolBlend {blendOption=Linear}
    boundBone {boneName=dyn_skirt_side_fr_01_l}
    componentAppliedLinearAccClamp=<7000 7000 7000>
    componentLinearAccScale=<8 8 8>
    componentLinearVelScale=<8 8 8>
    gravityOverride {z=-980}
    physicsBodyDefinitions [{
      collisionType=CustomSphere
      sphereCollisionRadius=3.3
      boundBone {boneName=dyn_skirt_side_fr_01_l}
      boxExtents=<10 10 10>
      constraintSetup {angularLimitsMax=<3 0 7>, angularLimitsMin=<-3 -25 -7>, angularTarget=<1 0 0>, linearAxesMax=<0 0 1>}
      localJointOffset=<0.5 0 0>
     }]
    relativeSpaceBone {boneName=pelvis}
    sphericalLimits [{limitRadius=4.5, drivingBone {boneName=thigh_l}, sphereLocalOffset=<-7 0 0>}]
   }
   n7 = Anim Dynamics - Bone: dyn_skirt_side_bk_01_l!(ComponentPose=@n6, bAlphaBoolEnabled:bool=@n3, ExternalForce="0,0,0") -> (Pose) {
    alphaInputType=Bool
    angularDampingOverride=0.999
    angularSpringConstant=300
    bAngularSpring=true
    bLinearSpring=true
    bOverrideAngularDamping=true
    bOverrideLinearDamping=true
    bUseGravityOverride=true
    linearDampingOverride=0.999
    linearSpringConstant=50
    simulationSpace=RootRelative
    alphaBoolBlend {blendOption=Linear}
    boundBone {boneName=dyn_skirt_side_bk_01_l}
    componentAppliedLinearAccClamp=<7000 7000 7000>
    componentLinearAccScale=<8 8 8>
    componentLinearVelScale=<8 8 8>
    gravityOverride {z=-980}
    physicsBodyDefinitions [{
      boundBone {boneName=dyn_skirt_side_bk_01_l}
      boxExtents=<10 10 10>
      constraintSetup {angularLimitsMax=<3 0 7>, angularLimitsMin=<-3 -25 -7>, angularTarget=<1 0 0>, linearAxesMax=<0 0 1>}
      localJointOffset=<0.5 0 0>
     }]
    relativeSpaceBone {boneName=pelvis}
   }
   n8 = Anim Dynamics - Bone: dyn_skirt_bk_01_l!(ComponentPose=@n7, bAlphaBoolEnabled:bool=@n3, ExternalForce="0,0,0") -> (Pose) {
    alphaInputType=Bool
    angularDampingOverride=0.999
    angularSpringConstant=300
    bAngularSpring=true
    bLinearSpring=true
    bOverrideAngularDamping=true
    bOverrideLinearDamping=true
    bUseGravityOverride=true
    linearDampingOverride=0.999
    linearSpringConstant=50
    simulationSpace=RootRelative
    alphaBoolBlend {blendOption=Linear}
    boundBone {boneName=dyn_skirt_bk_01_l}
    componentAppliedLinearAccClamp=<7000 7000 7000>
    componentLinearAccScale=<8 8 8>
    componentLinearVelScale=<8 8 8>
    gravityOverride {z=-980}
    physicsBodyDefinitions [{
      boundBone {boneName=dyn_skirt_bk_01_l}
      boxExtents=<10 10 10>
      constraintSetup {angularLimitsMax=<0 0 10>, angularLimitsMin=<0 -25 -10>, angularTarget=<1 0 0>, linearAxesMax=<0 0 1>}
      localJointOffset=<0.5 0 0>
     }]
    relativeSpaceBone {boneName=pelvis}
   }
   n9 = Anim Dynamics - Bone: dyn_skirt_fr_01_r!(ComponentPose=@n8, bAlphaBoolEnabled:bool=@n3, ExternalForce="0,0,0") -> (Pose) {
    alphaInputType=Bool
    angularDampingOverride=0.999
    angularSpringConstant=300
    bAngularSpring=true
    bLinearSpring=true
    bOverrideAngularDamping=true
    bOverrideLinearDamping=true
    bUseGravityOverride=true
    bUseSphericalLimits=true
    linearDampingOverride=0.999
    linearSpringConstant=50
    simulationSpace=RootRelative
    alphaBoolBlend {blendOption=Linear}
    boundBone {boneName=dyn_skirt_fr_01_r}
    componentAppliedLinearAccClamp=<7000 7000 7000>
    componentLinearAccScale=<8 8 8>
    componentLinearVelScale=<8 8 8>
    gravityOverride {z=-980}
    physicsBodyDefinitions [{
      collisionType=CustomSphere
      sphereCollisionRadius=4.3
      boundBone {boneName=dyn_skirt_fr_01_r}
      boxExtents=<10 10 10>
      constraintSetup {angularLimitsMax=<3 0 7>, angularLimitsMin=<-3 -25 -7>, angularTarget=<1 0 0>, linearAxesMax=<0 0 1>}
      localJointOffset=<0.5 0 0>
     }]
    relativeSpaceBone {boneName=pelvis}
    sphericalLimits [
     {limitRadius=4.5, drivingBone {boneName=thigh_r}, sphereLocalOffset=<5 0 0>}
     {limitRadius=3.5, drivingBone {boneName=thigh_r}, sphereLocalOffset=<5 0 -5>}
    ]
   }
   n10 = Anim Dynamics - Bone: dyn_skirt_fr_02_r!(ComponentPose=@n9, bAlphaBoolEnabled:bool=@n3, ExternalForce="0,0,0") -> (Pose) {
    alphaInputType=Bool
    angularDampingOverride=0.999
    angularSpringConstant=300
    bAngularSpring=true
    bLinearSpring=true
    bOverrideAngularDamping=true
    bOverrideLinearDamping=true
    bUseGravityOverride=true
    bUseSphericalLimits=true
    linearDampingOverride=0.999
    linearSpringConstant=50
    simulationSpace=RootRelative
    alphaBoolBlend {blendOption=Linear}
    boundBone {boneName=dyn_skirt_fr_02_r}
    componentAppliedLinearAccClamp=<7000 7000 7000>
    componentLinearAccScale=<8 8 8>
    componentLinearVelScale=<8 8 8>
    gravityOverride {z=-980}
    physicsBodyDefinitions [{
      collisionType=CustomSphere
      sphereCollisionRadius=4.3
      boundBone {boneName=dyn_skirt_fr_02_r}
      boxExtents=<10 10 10>
      constraintSetup {angularLimitsMax=<3 0 7>, angularLimitsMin=<-3 -25 -7>, angularTarget=<1 0 0>, linearAxesMax=<0 0 1>}
      localJointOffset=<0.5 0 0>
     }]
    relativeSpaceBone {boneName=pelvis}
    sphericalLimits [
     {limitRadius=5, drivingBone {boneName=thigh_r}, sphereLocalOffset=<10 0 0>}
     {limitRadius=4, drivingBone {boneName=thigh_r}, sphereLocalOffset=<10 0 -5>}
    ]
   }
   n11 = Anim Dynamics - Bone: dyn_skirt_side_fr_01_r!(ComponentPose=@n10, bAlphaBoolEnabled:bool=@n3, ExternalForce="0,0,0") -> (Pose) {
    alphaInputType=Bool
    angularDampingOverride=0.999
    angularSpringConstant=300
    bAngularSpring=true
    bLinearSpring=true
    bOverrideAngularDamping=true
    bOverrideLinearDamping=true
    bUseGravityOverride=true
    bUseSphericalLimits=true
    linearDampingOverride=0.999
    linearSpringConstant=50
    simulationSpace=RootRelative
    alphaBoolBlend {blendOption=Linear}
    boundBone {boneName=dyn_skirt_side_fr_01_r}
    componentAppliedLinearAccClamp=<7000 7000 7000>
    componentLinearAccScale=<8 8 8>
    componentLinearVelScale=<8 8 8>
    gravityOverride {z=-980}
    physicsBodyDefinitions [{
      collisionType=CustomSphere
      sphereCollisionRadius=3.5
      boundBone {boneName=dyn_skirt_side_fr_01_r}
      boxExtents=<10 10 10>
      constraintSetup {angularLimitsMax=<3 0 7>, angularLimitsMin=<-3 -25 -7>, angularTarget=<1 0 0>, linearAxesMax=<0 0 1>}
      localJointOffset=<0.5 0 0>
     }]
    relativeSpaceBone {boneName=pelvis}
    sphericalLimits [{limitRadius=4, drivingBone {boneName=thigh_r}, sphereLocalOffset=<5 0 0>}]
   }
   n12 = Anim Dynamics - Bone: dyn_skirt_side_bk_01_r!(ComponentPose=@n11, bAlphaBoolEnabled:bool=@n3, ExternalForce="0,0,0") -> (Pose) {
    alphaInputType=Bool
    angularDampingOverride=0.999
    angularSpringConstant=300
    bAngularSpring=true
    bLinearSpring=true
    bOverrideAngularDamping=true
    bOverrideLinearDamping=true
    bUseGravityOverride=true
    linearDampingOverride=0.999
    linearSpringConstant=50
    simulationSpace=RootRelative
    alphaBoolBlend {blendOption=Linear}
    boundBone {boneName=dyn_skirt_side_bk_01_r}
    componentAppliedLinearAccClamp=<7000 7000 7000>
    componentLinearAccScale=<8 8 8>
    componentLinearVelScale=<8 8 8>
    gravityOverride {z=-980}
    physicsBodyDefinitions [{
      boundBone {boneName=dyn_skirt_side_bk_01_r}
      boxExtents=<10 10 10>
      constraintSetup {angularLimitsMax=<3 0 7>, angularLimitsMin=<-3 -25 -7>, angularTarget=<1 0 0>, linearAxesMax=<0 0 1>}
      localJointOffset=<0.5 0 0>
     }]
    relativeSpaceBone {boneName=pelvis}
   }
   n13 = Anim Dynamics - Bone: dyn_skirt_bk_01_r!(ComponentPose=@n12, bAlphaBoolEnabled:bool=@n3, ExternalForce="0,0,0") -> (Pose) {
    alphaInputType=Bool
    angularDampingOverride=0.999
    angularSpringConstant=300
    bAngularSpring=true
    bLinearSpring=true
    bOverrideAngularDamping=true
    bOverrideLinearDamping=true
    bUseGravityOverride=true
    linearDampingOverride=0.999
    linearSpringConstant=50
    simulationSpace=RootRelative
    alphaBoolBlend {blendOption=Linear}
    boundBone {boneName=dyn_skirt_bk_01_r}
    componentAppliedLinearAccClamp=<7000 7000 7000>
    componentLinearAccScale=<8 8 8>
    componentLinearVelScale=<8 8 8>
    gravityOverride {z=-980}
    physicsBodyDefinitions [{
      boundBone {boneName=dyn_skirt_bk_01_r}
      boxExtents=<10 10 10>
      constraintSetup {angularLimitsMax=<0 0 10>, angularLimitsMin=<0 -25 -10>, angularTarget=<1 0 0>, linearAxesMax=<0 0 1>}
      localJointOffset=<0.5 0 0>
     }]
    relativeSpaceBone {boneName=pelvis}
   }
   n14 = RigidBody(ComponentPose=@n13, OverrideWorldGravity="0,0,-980") {
    bClampLinearTranslationLimitToRefPose=true
    bOverrideWorldGravity=true
    overridePhysicsAsset=&/F_MED_BlowWire_Physics_belt
    simulationSpace=BaseBoneSpace
    alphaBoolBlend {blendOption=Linear}
    baseBoneRef {boneName=pelvis}
    componentAppliedLinearAccClamp=<5000 5000 5000>
    componentLinearAccScale=<0.8 0.8 0.8>
    componentLinearVelScale=<0.8 0.8 0.8>
    overrideWorldGravity {z=-980}
    simSpaceSettings {
     maxAngularAcceleration=500
     maxAngularVelocity=3
     maxLinearAcceleration=400
     maxLinearVelocity=250
     velocityScaleZ=0
     worldAlpha=1
     externalLinearDragV=<0.1 0.1>
    }
   }
   n15 = Anim Dynamics - Bone: dyn_pouch_l_A_MSTR(ComponentPose=@n14, Alpha=1, ExternalForce="0,0,0") {
    angularDampingOverride=1
    angularSpringConstant=300
    bAngularSpring=true
    bOverrideAngularDamping=true
    bUseGravityOverride=true
    simulationSpace=RootRelative
    alphaBoolBlend {blendOption=Linear}
    boundBone {boneName=dyn_pouch_l_A_MSTR}
    componentAppliedLinearAccClamp=<5000 5000 5000>
    componentLinearAccScale=<2 2 2>
    componentLinearVelScale=<2 2 2>
    gravityOverride {z=-980}
    physicsBodyDefinitions [{
      boundBone {boneName=dyn_pouch_l_A_MSTR}
      boxExtents=<10 10 10>
      constraintSetup {angularLimitsMax=<5 5 5>, angularLimitsMin=<-5 -5 -5>, angularTarget=<0 0 0>}
      localJointOffset=<0.2 0 0>
     }]
   }
   n16 = Anim Dynamics - Bone: dyn_pouch_l_B_MSTR(ComponentPose=@n15, Alpha=1, ExternalForce="0,0,0") {
    angularDampingOverride=1
    angularSpringConstant=300
    bAngularSpring=true
    bOverrideAngularDamping=true
    bUseGravityOverride=true
    simulationSpace=RootRelative
    alphaBoolBlend {blendOption=Linear}
    boundBone {boneName=dyn_pouch_l_B_MSTR}
    componentAppliedLinearAccClamp=<5000 5000 5000>
    componentLinearAccScale=<2 2 2>
    componentLinearVelScale=<2 2 2>
    gravityOverride {z=-980}
    physicsBodyDefinitions [{
      boundBone {boneName=dyn_pouch_l_B_MSTR}
      boxExtents=<10 10 10>
      constraintSetup {angularLimitsMax=<5 5 5>, angularLimitsMin=<-5 -5 -5>, angularTarget=<0 0 0>}
      localJointOffset=<0.2 0 0>
     }]
   }
   n17 = Anim Dynamics - Bone: dyn_knife_01_l(ComponentPose=@n16, Alpha=1, ExternalForce="0,0,0") {
    angularDampingOverride=1
    angularSpringConstant=300
    bAngularSpring=true
    bOverrideAngularDamping=true
    bUseGravityOverride=true
    simulationSpace=RootRelative
    alphaBoolBlend {blendOption=Linear}
    boundBone {boneName=dyn_knife_01_l}
    componentAppliedLinearAccClamp=<5000 5000 5000>
    componentLinearAccScale=<1 1 1>
    componentLinearVelScale=<1 1 1>
    gravityOverride {z=-980}
    physicsBodyDefinitions [{
      boundBone {boneName=dyn_knife_01_l}
      boxExtents=<10 10 10>
      constraintSetup {angularLimitsMax=<3 0 3>, angularLimitsMin=<-3 0 -3>, angularTarget=<0 0 0>}
      localJointOffset=<0.1 0 0>
     }]
   }
   n18 = Anim Dynamics - Bone: Dyn_collar_L(ComponentPose=@n17, Alpha=1, ExternalForce="0,0,0") {
    angularDampingOverride=1
    angularSpringConstant=300
    bAngularSpring=true
    bOverrideAngularDamping=true
    bUseGravityOverride=true
    simulationSpace=RootRelative
    alphaBoolBlend {blendOption=Linear}
    boundBone {boneName=Dyn_collar_L}
    componentAppliedLinearAccClamp=<5000 5000 5000>
    componentLinearAccScale=<2 2 2>
    componentLinearVelScale=<2 2 2>
    gravityOverride {z=-980}
    physicsBodyDefinitions [{
      boundBone {boneName=Dyn_collar_L}
      boxExtents=<10 10 10>
      constraintSetup {angularLimitsMax=<0 0 15>, angularLimitsMin=<0 -15 -15>, angularTarget=<0 0 0>}
      localJointOffset=<1 0 0>
     }]
    relativeSpaceBone {boneName=spine_04}
   }
   n19 = Anim Dynamics - Bone: Dyn_collar_R(ComponentPose=@n18, Alpha=1, ExternalForce="0,0,0") {
    angularDampingOverride=1
    angularSpringConstant=300
    bAngularSpring=true
    bOverrideAngularDamping=true
    bUseGravityOverride=true
    simulationSpace=RootRelative
    alphaBoolBlend {blendOption=Linear}
    boundBone {boneName=Dyn_collar_R}
    componentAppliedLinearAccClamp=<5000 5000 5000>
    componentLinearAccScale=<2 2 2>
    componentLinearVelScale=<2 2 2>
    gravityOverride {z=-980}
    physicsBodyDefinitions [{
      boundBone {boneName=Dyn_collar_R}
      boxExtents=<10 10 10>
      constraintSetup {angularLimitsMax=<0 0 15>, angularLimitsMin=<0 -15 -15>, angularTarget=<0 0 0>}
      localJointOffset=<1 0 0>
     }]
    relativeSpaceBone {boneName=spine_04}
   }
   Output Pose(Result=Component To Local(ComponentPose=@n19)) {layerGroup=None}
   n20 = Control Rig!(
    Source=Copy Pose From Mesh() {bUseAttachedParent=true}
    Alpha
    LR_Movement:double=Get Left Right Movement() -> (Left Right Movement:double)
    FB_Movement:double=Get Forward Backward Movement() -> (Forward Backward Movement:double)
    TacticalSprinting:bool=Get bIsTacticalSprinting()
    Emotes:bool=Get Emotes() -> (Emotes:bool)
    RdanOFF:bool=Get RdanOFF() -> (RdanOFF:bool)
    Lod1:bool=Get LOD1() -> (LOD1:bool)
    Lod2:bool=Get LOD2() -> (LOD2:bool)
    High Fps:double=Get High FPS() -> (High FPS:double)
    Sprint and LowFPS:bool=Get Sprint and LowFPS()
   ) -> (Pose) {
    alphaBoolBlend {blendOption=Linear}
    controlRigAssetReference {blueprintRigClass=&/F_MED_BlowWire_CtrlRig.F_MED_BlowWire_CtrlRig_C}
   }
   Save cached pose 'Base'(Pose=@n20) {cacheName=Base}
  ]}
 EventGraph {nodes [
   n1 = Event BlueprintUpdateAnimation()
   > n2 = Set Delta Time X(Delta Time X=@n1.DeltaTimeX)
   > Set LOW FPS(LOW FPS:float=float / float(A=1, B=@n2)) -> (Output_Get:float)
   > n3 = Sequence()
   > n4 = Gravity Override State Function DEPRECATED(CharacterPartAnimInstance=Self-Reference(), GameTimeInSeconds=GetGameTimeInSeconds(), GravityOverrideParams=Get Gravity Override Params())
   > Set Gravity Override(Gravity Override=@n4)
   > Set RdanOFF(RdanOFF:bool=NOT Boolean(A=Get bRigidBodyEnabled())) -> (Output_Get:bool)
   n5 = Get CurrentLOD()
   > Set LOD1(LOD1:bool="Equal (Integer)"(A=@n5, B=1)) -> (Output_Get:bool)
   > Set LOD2(LOD2:bool="integer >= integer"(A=@n5, B=2)) -> (Output_Get:bool)
   > Set High FPS(High FPS:double=MapRangeClamped(Value=Get LOW FPS() -> (LOW FPS:float), InRangeA=35, InRangeB=40)) -> (Output_Get:double)
   n6 = NOT Boolean(A=Get bIsCrouching())
   n7 = Get bIsMoving2D()
   n8 = FInterpTo(Current=MapRangeClamped(Value=Get LOW FPS(), InRangeA=50, InRangeB=40), Target=0, DeltaTime=Get Delta Time X(), InterpSpeed=15)
   > Set Sprint and LowFPS(Sprint and LowFPS:bool=OR Boolean!(A=Get bIsTacticalSprinting(), B="Equal (Float)"(A=@n8, B=1), C:bool=AND Boolean(A=@n6, B=@n7)) -> (ReturnValue)) -> (Output_Get:bool)
   n9 = PlayerMovement(execute=@n3.then_1, AnimInstance=Self-Reference(), XDotProductCoefficient=1, YDotProductCoefficient=1, ZDotProductCoefficient=1)
   > Set Left Right Movement(Left Right Movement:double=@n9) -> (Output_Get:double)
   > Set Forward Backward Movement(Forward Backward Movement:double=@n9.YMovement) -> (Output_Get:double)
   Event OnBeginPlayingEmote() {is=title K2Node_Event}
   > Set Emotes(Emotes:bool=true) -> (Output_Get:bool)
   Event OnEndPlayingEmote() {is=title K2Node_Event}
   > Set Emotes(Emotes=false)
  ]}
]
variables [
 Delta Time X {type=number}
 Emotes {type=boolean}
 Forward Backward Movement {type=number}
 Gravity Override {type=Vector}
 Gravity Override Params {type=GravityOverrideParamsStruct}
 High FPS {type=number}
 Left Right Movement {type=number}
 LOD1 {type=boolean}
 LOD2 {type=boolean}
 LOW FPS {type=number}
 RdanOFF {type=boolean}
 Sprint and LowFPS {type=boolean}
]
