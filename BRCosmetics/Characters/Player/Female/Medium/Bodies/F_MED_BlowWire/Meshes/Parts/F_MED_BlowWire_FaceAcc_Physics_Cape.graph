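# Physics asset for the female medium BlowWire FaceAcc part, simulating hair bones despite the Cape name.
# Capsules cover front left/right and back center/left/right hair chains, plus head, neck, upper spine, clavicles and upper arms.
# Front chains and back center root at the head with free rotation; back side chains branch off the center.
# Hair joints mostly have tight angular limits, while free-rotation cross-constraints tie lower back side strands to the center.
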PhysicsAsset /BRCosmetics/Characters/Player/Female/Medium/Bodies/F_MED_BlowWire/Meshes/Parts/F_MED_BlowWire_FaceAcc_Physics_Cape
bodies {
 Base_hair_Back_01_C [{shapeType=Capsule, length=0.5, radius=0.5, shapeName=Base_hair_Back_01_C_capsule}]
 Base_hair_Back_02_C [{shapeType=Capsule, length=0.5, radius=0.5, shapeName=Base_hair_Back_02_C_capsule}]
 Base_hair_Back_l [{shapeType=Capsule, length=0.5, radius=0.5, shapeName=Base_hair_Back_l_capsule}]
 Base_hair_Back_r [{shapeType=Capsule, length=0.5, radius=0.5, shapeName=Base_hair_Back_r_capsule}]
 Base_hair_fr_l [{shapeType=Capsule, length=0.5, radius=0.5, shapeName=Base_hair_fr_l_capsule}]
 Base_hair_fr_r [{shapeType=Capsule, length=0.5, radius=0.5, shapeName=Base_hair_fr_r_capsule}]
 Dyn_hair_Back_01_C [{shapeType=Capsule, length=3, radius=2.5, shapeName=Dyn_hair_Back_01_C_capsule, center=<2.5029 0 0>, rotation=<rot 90 0 0>}]
 Dyn_hair_Back_01_l [{shapeType=Capsule, length=3, radius=2.5, shapeName=Dyn_hair_Back_01_l_capsule, center=<2.5029 0 0>, rotation=<rot 90 0 0>}]
 Dyn_hair_Back_01_r [{shapeType=Capsule, length=3, radius=2.5, shapeName=Dyn_hair_Back_01_r_capsule, center=<2.5029 0 0>, rotation=<rot 90 0 0>}]
 Dyn_hair_Back_02_C [{shapeType=Capsule, length=3, radius=2.5, shapeName=Dyn_hair_Back_02_C_capsule, center=<2.5029 0 0>, rotation=<rot 90 0 0>}]
 Dyn_hair_Back_02_l [{shapeType=Capsule, length=3, radius=2.5, shapeName=Dyn_hair_Back_02_l_capsule, center=<2.5029 0 0>, rotation=<rot 90 0 0>}]
 Dyn_hair_Back_02_r [{shapeType=Capsule, length=2.7806, radius=2.7194, shapeName=Dyn_hair_Back_02_r_capsule, center=<2.5029 0 0>, rotation=<rot 90 0 0>}]
 Dyn_hair_Back_03_C [{shapeType=Capsule, length=3, radius=2.5, shapeName=Dyn_hair_Back_03_C_capsule, center=<2.5029 0 0>, rotation=<rot 90 0 0>}]
 Dyn_hair_Back_03_l [{shapeType=Capsule, length=2.2735, radius=1.9673, shapeName=Dyn_hair_Back_03_l_capsule, center=<2.5029 0 0>, rotation=<rot 90 0 0>}]
 Dyn_hair_Back_03_r [{shapeType=Capsule, length=2.7806, radius=2.7194, shapeName=Dyn_hair_Back_03_r_capsule, center=<2.5029 0 0>, rotation=<rot 90 0 0>}]
 Dyn_hair_Back_04_C [{shapeType=Capsule, length=3, radius=2.5, shapeName=Dyn_hair_Back_04_C_capsule, center=<2.5029 0 0>, rotation=<rot 90 0 0>}]
 Dyn_hair_Back_04_l [{shapeType=Capsule, length=2.8866, radius=2.6134, shapeName=Dyn_hair_Back_04_l_capsule, center=<2.5029 0 0>, rotation=<rot 90 0 0>}]
 Dyn_hair_Back_04_r [{shapeType=Capsule, length=2.1309, radius=2.02, shapeName=Dyn_hair_Back_04_r_capsule, center=<2.5029 0 0>, rotation=<rot 90 0 0>}]
 Dyn_hair_Back_05_C [{shapeType=Capsule, length=3.45, radius=2.05, shapeName=Dyn_hair_Back_05_C_capsule, center=<2.5029 0 0>, rotation=<rot 90 0 0>}]
 Dyn_hair_Back_05_r [{
   shapeType=Capsule
   length=1.0624
   radius=2.1175
   shapeName=Dyn_hair_Back_05_r_capsule
   center=<2.6668 0.01 0.0437>
   rotation=<rot 0 75.6732 -89.8158>
  }]
 Dyn_hair_Back_06_C [{shapeType=Capsule, length=2.284, radius=2.0021, shapeName=Dyn_hair_Back_06_C_capsule, center=<2.5029 0 0>, rotation=<rot 90 0 0>}]
 Dyn_hair_fr_01_l [{shapeType=Capsule, length=2.6707, radius=2.8293, shapeName=Dyn_hair_fr_01_l_capsule, center=<2.5029 0 0>, rotation=<rot 90 0 0>}]
 Dyn_hair_fr_01_r [{shapeType=Capsule, length=2.9398, radius=2.2495, shapeName=Dyn_hair_fr_01_r_capsule, center=<2.5029 0 0>, rotation=<rot 90 0 0>}]
 Dyn_hair_fr_02_l [{shapeType=Capsule, length=3, radius=2.5, shapeName=Dyn_hair_fr_02_l_capsule, center=<2.5029 0 0.8205>, rotation=<rot 90 0 0>}]
 Dyn_hair_fr_02_r [{shapeType=Capsule, length=3, radius=2.5, shapeName=Dyn_hair_fr_02_r_capsule, center=<2.5029 0 0>, rotation=<rot 90 0 0>}]
 Dyn_hair_fr_03_l [{shapeType=Capsule, length=4.0689, radius=1.4311, shapeName=Dyn_hair_fr_03_l_capsule, center=<3.5874 0 0>, rotation=<rot 90 0 0>}]
 Dyn_hair_fr_03_r [{
   shapeType=Capsule
   length=2.6355
   radius=1.9201
   shapeName=Dyn_hair_fr_03_r_capsule
   center=<2.5029 0 0.727>
   rotation=<rot 90 179.4218 169.4219>
  }]
 Dyn_hair_fr_04_l [{shapeType=Capsule, length=2.9518, radius=1.4145, shapeName=Dyn_hair_fr_04_l_capsule, center=<2.5029 0 0.1155>, rotation=<rot 90 0 0>}]
 Dyn_hair_fr_04_r [{shapeType=Capsule, length=2.0935, radius=1.3312, shapeName=Dyn_hair_fr_04_r_capsule, center=<2.5029 0 0>, rotation=<rot 90 0 0>}]
 clavicle_l [{shapeType=Capsule, length=8.4167, radius=5.5176, shapeName=clavicle_l_capsule, center=<5.2815 0 -0.4621>, rotation=<rot -0.0003 -90 95>}]
 clavicle_r [{shapeType=Capsule, length=8.6948, radius=5.2395, shapeName=clavicle_r_capsule, center=<-3.8532 0 0.3371>, rotation=<rot -0.0003 -90 95>}]
 head [{shapeType=Capsule, length=3.9867, radius=6.5852, shapeName=head_capsule, center=<2.1995 0 0>, rotation=<rot 90 0 0>}]
 neck_01 [{shapeType=Capsule, length=3.9101, radius=5.1607, shapeName=neck_01_capsule, rotation=<rot 90 0 0>}]
 neck_02 [{shapeType=Capsule, length=2.5289, radius=5.5899, shapeName=neck_02_capsule, rotation=<rot 90 0 0>}]
 spine_05 [{shapeType=Capsule, length=13.1668, radius=8.5461, shapeName=spine_05_capsule}]
 upperarm_l [{shapeType=Capsule, length=14.0224, radius=3.8659, shapeName=upperarm_l_capsule, center=<10.2301 0 0>, rotation=<rot 90 0 0>}]
 upperarm_r [{shapeType=Capsule, length=15.1661, radius=3.8659, shapeName=upperarm_r_capsule, center=<-9.4478 0 0>, rotation=<rot 90 0 0>}]
}
constraints [
 {bone1Name=Dyn_hair_fr_04_r, bone2Name=Dyn_hair_fr_01_r, swing1LimitDegrees=15, swing2LimitDegrees=15, twistLimitDegrees=0}
 {
  bone1Name=neck_02
  bone2Name=neck_01
  swing1LimitDegrees=45
  swing1Motion=Free
  swing2LimitDegrees=45
  swing2Motion=Free
  twistLimitDegrees=45
  twistMotion=Free
 }
 {
  bone1Name=head
  bone2Name=neck_02
  swing1LimitDegrees=45
  swing1Motion=Free
  swing2LimitDegrees=45
  swing2Motion=Free
  twistLimitDegrees=45
  twistMotion=Free
 }
 {
  bone1Name=Base_hair_Back_01_C
  bone2Name=head
  swing1LimitDegrees=45
  swing1Motion=Free
  swing2LimitDegrees=45
  swing2Motion=Free
  twistLimitDegrees=45
  twistMotion=Free
 }
 {bone1Name=Dyn_hair_Back_01_C, bone2Name=Base_hair_Back_01_C, swing1LimitDegrees=10, swing2LimitDegrees=5, twistLimitDegrees=10}
 {bone1Name=Base_hair_Back_02_C, bone2Name=Dyn_hair_Back_01_C, swing1LimitDegrees=10, swing2LimitDegrees=0, twistLimitDegrees=5}
 {bone1Name=Dyn_hair_Back_02_C, bone2Name=Base_hair_Back_02_C, swing1LimitDegrees=10, swing2LimitDegrees=0, twistLimitDegrees=5}
 {bone1Name=Dyn_hair_Back_03_C, bone2Name=Dyn_hair_Back_02_C, swing1LimitDegrees=8, swing2LimitDegrees=0, twistLimitDegrees=0}
 {bone1Name=Dyn_hair_Back_04_C, bone2Name=Dyn_hair_Back_03_C, swing1LimitDegrees=10, swing2LimitDegrees=10, twistLimitDegrees=0}
 {bone1Name=Dyn_hair_Back_05_C, bone2Name=Dyn_hair_Back_04_C, swing1LimitDegrees=10, swing2LimitDegrees=10, twistLimitDegrees=0}
 {bone1Name=Dyn_hair_Back_06_C, bone2Name=Dyn_hair_Back_04_C, swing1LimitDegrees=10, swing2LimitDegrees=10, twistLimitDegrees=0}
 {bone1Name=Base_hair_Back_l, bone2Name=Dyn_hair_Back_01_C, swing1LimitDegrees=10, swing2LimitDegrees=5, twistLimitDegrees=10}
 {bone1Name=Dyn_hair_Back_01_l, bone2Name=Base_hair_Back_l, swing1LimitDegrees=10, swing2LimitDegrees=5, twistLimitDegrees=10}
 {bone1Name=Dyn_hair_Back_02_l, bone2Name=Dyn_hair_Back_01_l, swing1LimitDegrees=5, swing2LimitDegrees=0, twistLimitDegrees=5}
 {bone1Name=Dyn_hair_Back_03_l, bone2Name=Dyn_hair_Back_02_l, swing1LimitDegrees=0, swing2LimitDegrees=0, twistLimitDegrees=0}
 {bone1Name=Dyn_hair_Back_04_l, bone2Name=Dyn_hair_Back_01_l, swing1LimitDegrees=5, swing2LimitDegrees=0, twistLimitDegrees=0}
 {bone1Name=Base_hair_Back_r, bone2Name=Dyn_hair_Back_01_C, swing1LimitDegrees=20, swing2LimitDegrees=20, twistLimitDegrees=10}
 {bone1Name=Dyn_hair_Back_01_r, bone2Name=Base_hair_Back_r, swing1LimitDegrees=10, swing2LimitDegrees=5, twistLimitDegrees=10}
 {bone1Name=Dyn_hair_Back_02_r, bone2Name=Dyn_hair_Back_01_r, swing1LimitDegrees=10, swing2LimitDegrees=0, twistLimitDegrees=0}
 {bone1Name=Dyn_hair_Back_03_r, bone2Name=Dyn_hair_Back_02_r, swing1LimitDegrees=5, swing2LimitDegrees=0, twistLimitDegrees=0}
 {bone1Name=Dyn_hair_Back_04_r, bone2Name=Dyn_hair_Back_03_r, swing1LimitDegrees=5, swing2LimitDegrees=0, twistLimitDegrees=0}
 {
  bone1Name=Base_hair_fr_l
  bone2Name=head
  swing1LimitDegrees=45
  swing1Motion=Free
  swing2LimitDegrees=45
  swing2Motion=Free
  twistLimitDegrees=45
  twistMotion=Free
 }
 {bone1Name=Dyn_hair_fr_01_l, bone2Name=Base_hair_fr_l, swing1LimitDegrees=30, swing2LimitDegrees=30, twistLimitDegrees=5}
 {bone1Name=Dyn_hair_fr_02_l, bone2Name=Dyn_hair_fr_01_l, swing1LimitDegrees=25, swing2LimitDegrees=25, twistLimitDegrees=5}
 {bone1Name=Dyn_hair_fr_03_l, bone2Name=Dyn_hair_fr_02_l, swing1LimitDegrees=15, swing2LimitDegrees=15, twistLimitDegrees=0}
 {bone1Name=Dyn_hair_fr_04_l, bone2Name=Dyn_hair_fr_01_l, swing1LimitDegrees=15, swing2LimitDegrees=15, twistLimitDegrees=0}
 {
  bone1Name=Base_hair_fr_r
  bone2Name=head
  swing1LimitDegrees=45
  swing1Motion=Free
  swing2LimitDegrees=45
  swing2Motion=Free
  twistLimitDegrees=45
  twistMotion=Free
 }
 {bone1Name=Dyn_hair_fr_01_r, bone2Name=Base_hair_fr_r, swing1LimitDegrees=30, swing2LimitDegrees=30, twistLimitDegrees=10}
 {bone1Name=Dyn_hair_fr_02_r, bone2Name=Dyn_hair_fr_01_r, swing1LimitDegrees=25, swing2LimitDegrees=25, twistLimitDegrees=5}
 {bone1Name=Dyn_hair_fr_03_r, bone2Name=Dyn_hair_fr_02_r, swing1LimitDegrees=15, swing2LimitDegrees=15, twistLimitDegrees=0}
 {
  bone1Name=neck_01
  bone2Name=spine_05
  swing1LimitDegrees=45
  swing1Motion=Free
  swing2LimitDegrees=45
  swing2Motion=Free
  twistLimitDegrees=45
  twistMotion=Free
 }
 {
  bone1Name=clavicle_l
  bone2Name=spine_05
  swing1LimitDegrees=45
  swing1Motion=Free
  swing2LimitDegrees=45
  swing2Motion=Free
  twistLimitDegrees=45
  twistMotion=Free
 }
 {
  bone1Name=clavicle_r
  bone2Name=spine_05
  swing1LimitDegrees=45
  swing1Motion=Free
  swing2LimitDegrees=45
  swing2Motion=Free
  twistLimitDegrees=45
  twistMotion=Free
 }
 {
  bone1Name=upperarm_r
  bone2Name=clavicle_r
  swing1LimitDegrees=45
  swing1Motion=Free
  swing2LimitDegrees=45
  swing2Motion=Free
  twistLimitDegrees=45
  twistMotion=Free
 }
 {
  bone1Name=upperarm_l
  bone2Name=clavicle_l
  swing1LimitDegrees=45
  swing1Motion=Free
  swing2LimitDegrees=45
  swing2Motion=Free
  twistLimitDegrees=45
  twistMotion=Free
 }
 {bone1Name=Dyn_hair_Back_05_r, bone2Name=Dyn_hair_Back_02_r, swing1LimitDegrees=8, swing2LimitDegrees=8, twistLimitDegrees=0}
 {
  bone1Name=Dyn_hair_Back_03_l
  bone2Name=Dyn_hair_Back_04_C
  swing1LimitDegrees=45
  swing1Motion=Free
  swing2LimitDegrees=45
  swing2Motion=Free
  twistLimitDegrees=45
  twistMotion=Free
 }
 {
  bone1Name=Dyn_hair_Back_04_r
  bone2Name=Dyn_hair_Back_04_C
  swing1LimitDegrees=45
  swing1Motion=Free
  swing2LimitDegrees=45
  swing2Motion=Free
  twistLimitDegrees=45
  twistMotion=Free
 }
]
