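# Post-process style AnimBP for the BlowWire female face accessory, adding secondary motion on top of copied character poses.
# The base pose copies the attached parent, layering in Body mesh clavicles and upper arms.
# A backbling/cape check selects head-space RigidBody with a cape physics asset, or Control Rig (fed pickaxe and emote flags) into default-accessory RigidBody.
# Anim Dynamics then springs the pompom bone relative to the head.
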AnimBlueprint /BRCosmetics/Characters/Player/Female/Medium/Bodies/F_MED_BlowWire/Meshes/Parts/F_MED_BlowWire_FaceAcc_AnimBP
target_skeleton=&/F_MED_BlowWire_FaceAcc_Skeleton
graphs [
 AnimGraph {nodes [
   n1 = RigidBody!(ComponentPose=Local To Component(LocalPose=Use cached pose 'Base'()), Alpha, OverrideWorldGravity:Vector="0,0,0", ExternalForce) -> (Pose) {
    bClampLinearTranslationLimitToRefPose=true
    bOverrideWorldGravity=true
    overridePhysicsAsset=&/F_MED_BlowWire_FaceAcc_Physics_Cape
    simulationSpace=BaseBoneSpace
    alphaBoolBlend {blendOption=Linear}
    baseBoneRef {boneName=head}
    componentAppliedLinearAccClamp=<5000 5000 5000>
    componentLinearAccScale=<0.35 0.1 0.1>
    componentLinearVelScale=<0.35 0.1 0.1>
    simSpaceSettings {
     maxAngularAcceleration=500
     maxAngularVelocity=3
     maxLinearAcceleration=400
     maxLinearVelocity=250
     velocityScaleZ=0
     worldAlpha=1
     externalLinearDragV=<0.1 0.1>
    }
   }
   n2 = Component To Local(ComponentPose=@n1)
   n3 = Control Rig!(Source=Use cached pose 'Base'(), Alpha, pickaxeSwing:bool=Get pickaxeSwing() -> (pickaxeSwing:bool), PlayingEmote:bool=Get bIsPlayingEmote()) -> (Pose) {
    alphaBoolBlend {blendOption=Linear}
    controlRigAssetReference {blueprintRigClass=&/F_MED_BlowWire_FaceAcc_CtrlRig.F_MED_BlowWire_FaceAcc_CtrlRig_C}
   }
   n4 = RigidBody(ComponentPose=Local To Component(LocalPose=@n3), OverrideWorldGravity="0,0,0") {
    bClampLinearTranslationLimitToRefPose=true
    bOverrideWorldGravity=true
    overridePhysicsAsset=&/F_MED_BlowWire_FaceAcc_Physics
    simulationSpace=BaseBoneSpace
    alphaBoolBlend {blendOption=Linear}
    baseBoneRef {boneName=head}
    componentAppliedLinearAccClamp=<5000 5000 5000>
    componentLinearAccScale=<0.35 0.1 0.1>
    componentLinearVelScale=<0.35 0.1 0.1>
    simSpaceSettings {
     maxAngularAcceleration=500
     maxAngularVelocity=3
     maxLinearAcceleration=400
     maxLinearVelocity=250
     velocityScaleZ=0
     worldAlpha=1
     externalLinearDragV=<0.1 0.1>
    }
   }
   n5 = Component To Local(ComponentPose=@n4)
   n6 = Get Is Cape() -> (Is Cape:bool)
   n7 = Anim Dynamics - Bone: Dyn_pompom(
    ComponentPose=Local To Component(LocalPose=Blend Poses by bool(BlendPose_0=@n2, BlendPose_1=@n5, bActiveValue=@n6) {blendType=Linear})
    Alpha=1
    ExternalForce="0,0,0"
   ) {
    angularDampingOverride=1
    angularSpringConstant=600
    bAngularSpring=true
    bOverrideAngularDamping=true
    simulationSpace=BoneRelative
    alphaBoolBlend {blendOption=Linear}
    boundBone {boneName=Dyn_pompom}
    componentAppliedLinearAccClamp=<5000 5000 5000>
    componentLinearAccScale=<1 1 1>
    componentLinearVelScale=<1 1 1>
    physicsBodyDefinitions [{
      boundBone {boneName=Dyn_pompom}
      boxExtents=<10 10 10>
      constraintSetup {angularLimitsMax=<0 12 12>, angularLimitsMin=<0 -12 -12>, angularTarget=<0 0 0>}
      localJointOffset=<0.5 0 0>
     }]
    relativeSpaceBone {boneName=head}
   }
   Output Pose(Result=Component To Local(ComponentPose=@n7)) {layerGroup=None}
   n8 = Layered blend per bone(
    BasePose=Copy Pose From Mesh() {bUseAttachedParent=true}
    BlendPoses_0=Copy Pose From Mesh(SourceMeshComponent=Get Body() -> (Body:Skeletal Mesh Component Budgeted Object Reference))
   ) {layerSetup [{branchFilters [
       {blendDepth=2, boneName=clavicle_r}
       {blendDepth=2, boneName=clavicle_l}
       {blendDepth=2, boneName=upperarm_l}
       {blendDepth=2, boneName=upperarm_r}
      ]}]}
   Save cached pose 'Base'(Pose=@n8) {cacheName=Base}
  ]}
 EventGraph {nodes [
   Event BlueprintUpdateAnimation()
   > Sequence()
   > n1 = Gravity Override State Function DEPRECATED(CharacterPartAnimInstance=Self-Reference(), GameTimeInSeconds=GetGameTimeInSeconds(), GravityOverrideParams=Get Gravity Override Params())
   > Set Gravity Override(Gravity Override=@n1)
   > Set pickaxeSwing(pickaxeSwing:bool=Get bIsPlayingMeleeAnim()) -> (Output_Get:bool)
   n2 = CheckForBackbling()
   > Set Is Cape(Is Cape:bool=OR Boolean(A=@n2, B=@n2.bIsBackblingACape)) -> (Output_Get:bool)
   n3 = Event OnCharacterCustomizationLoaded()
   > Set Body(Body:Skeletal Mesh Component Budgeted Object Reference=Get Array Item with EFort Custom Part Type(Array=@n3.CustomMeshParts)) -> (Output_Get:Skeletal Mesh Component Budgeted Object Reference)
  ]}
]
variables [
 Body {type=SkeletalMeshComponentBudgeted}
 Current State Out {type=string, default=""}
 Gravity Override {type=Vector}
 Gravity Override Params {type=GravityOverrideParamsStruct}
 Is Cape {type=boolean}
 pickaxeSwing {type=boolean}
]
